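# Editor utility widget that copies checked bindings and their tracks from a source Level Sequence into a target one.
# Source and target pickers fill scroll boxes with a row per binding, offering copy and overwrite checkboxes on source rows.
# Copy opens the target, polls by timer until Sequencer has focus, then collects each binding into structs and pastes it.
# Helpers rebuild float, bool, transform, animation, spawn and Control Rig tracks, using Python for rig class.
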EditorUtilityWidgetBlueprint /AnimInEdTools/LevelSequenceAssetCopier/EUW_LevelSequenceAssetCopier
graphs [
 AddBoolSection {nodes [
   n1 = AddBoolSection!() -> (then, Track:Movie Scene Track Object Reference, BoolChannels:Array of Struct Asset Copier Bool Channel Structures) {is=title K2Node_FunctionEntry}
   > n2 = AddSection!(execute, Track:Movie Scene Track Object Reference=@n1) -> (then, ReturnValue:Movie Scene Section Object Reference)
   > n3 = Cast To MovieSceneBoolSection!(execute, Object=@n2) -> (then, CastFailed, AsMovie Scene Bool Section:Movie Scene Bool Section Object Reference) {is=title K2Node_DynamicCast}
   > SetStartFrameBounded!(execute, Section:Movie Scene Section Object Reference=@n3, bIsBounded:bool=false) -> (then)
   > SetEndFrameBounded!(execute, Section:Movie Scene Section Object Reference=@n3, bIsBounded:bool=false) -> (then)
   > SetBoolKeys!(
    execute
    self
    Section:Movie Scene Section Object Reference=@n3
    BoolChannels:Array of Struct Asset Copier Bool Channel Structures=Get BoolChannels() -> (BoolChannels:Array of Struct Asset Copier Bool Channel Structures)
   ) -> (then)
  ]}
 AddCtrlRigSection {nodes [
   n1 = AddCtrlRigSection!() -> (
    then
    Track:Movie Scene Track Object Reference
    InBlendType:EMovieSceneBlendType Enum=Invalid
    TrackRowName:Text
    FloatChannels:Array of Struct Asset Copier Float Channel Structures
    BoolChannels:Array of Struct Asset Copier Bool Channel Structures
    SectionRowIndex:int
   ) {is=title K2Node_FunctionEntry}
   > SetTrackRowDisplayName!(execute, Track:Movie Scene Track Object Reference=@n1, InName:"Text (by ref)"=@n1.TrackRowName, RowIndex:int=@n1.SectionRowIndex) -> (then)
   > n2 = AddSection(Track=@n1)
   > Execute Python Script!(
    execute
    PythonScript="import unreal\r\n\r\nsection = section\r\ncontrol_rig_class = rig\r\n\r\nsection.set_editor_property(\"control_rig_class\", control_rig_class)"
    section:Movie Scene Section Object Reference=@n2
    rig:Control Rig Class Reference=Get rig() -> (rig:Control Rig Class Reference)
   ) -> (then, ReturnValue) {inputs [section,rig]}
   > SetStartFrameBounded(Section=@n2, bIsBounded=false)
   > SetEndFrameBounded(Section=@n2, bIsBounded=false)
   > SetBlendType!(
    execute
    self:Movie Scene Section Object Reference=@n2
    InBlendType:EMovieSceneBlendType Enum=Get InBlendType() -> (InBlendType:EMovieSceneBlendType Enum=Invalid)
   ) -> (then)
   > SetRowIndex!(execute, self:Movie Scene Section Object Reference=@n2, NewRowIndex:int=Get SectionRowIndex() -> (SectionRowIndex:int=0)) -> (then)
   > SetFloatKeys!(
    execute
    self
    Section:Movie Scene Section Object Reference=@n2
    FloatChannels:Array of Struct Asset Copier Float Channel Structures=Get FloatChannels() -> (FloatChannels:Array of Struct Asset Copier Float Channel Structures)
   ) -> (then)
   > SetBoolKeys(Section=@n2, BoolChannels=Get BoolChannels())
  ]}
 AddFloatSection {nodes [
   n1 = AddFloatSection!() -> (
    then
    Track:Movie Scene Track Object Reference
    InBlendType:EMovieSceneBlendType Enum=Invalid
    FloatChannels:Array of Struct Asset Copier Float Channel Structures
    SectionRowIndex:int
   ) {is=title K2Node_FunctionEntry}
   > n2 = AddSection(Track=@n1)
   > n3 = Cast To MovieSceneFloatSection!(execute, Object=@n2) -> (then, CastFailed, AsMovie Scene Float Section:Movie Scene Float Section Object Reference) {is=title K2Node_DynamicCast}
   > SetStartFrameBounded(Section=@n3, bIsBounded=false)
   > SetEndFrameBounded(Section=@n3, bIsBounded=false)
   > SetRowIndex(self=@n3, NewRowIndex=@n1.SectionRowIndex)
   > SetBlendType(self=@n3, InBlendType=Get InBlendType() -> (InBlendType=Invalid))
   > SetFloatKeys(Section=@n3, FloatChannels=Get FloatChannels())
  ]}
 AddTransformSection {nodes [
   n1 = AddTransformSection!() -> (
    then
    InBlendType:EMovieSceneBlendType Enum=Invalid
    FloatData:Array of Struct Asset Copier Float Channel Structures
    SectionRowIndex:int
    Track:Movie Scene Track Object Reference
   ) {is=title K2Node_FunctionEntry}
   > n2 = AddSection(Track=@n1.Track)
   > SetStartFrameBounded(Section=@n2, bIsBounded=false)
   > SetEndFrameBounded(Section=@n2, bIsBounded=false)
   > SetBlendType(self=@n2, InBlendType=Get InBlendType() -> (InBlendType=Invalid))
   > SetRowIndex(self=@n2, NewRowIndex=Get SectionRowIndex() -> (SectionRowIndex=0))
   > SetTransformFloatKeys!(
    execute
    self
    Section:Movie Scene Section Object Reference=@n2
    FloatData:Array of Struct Asset Copier Float Channel Structures=Get FloatData() -> (FloatData:Array of Struct Asset Copier Float Channel Structures)
   ) -> (then)
  ]}
 CollectBindingData {nodes [
   n1 = CollectBindingData!() -> (then, InBinding:"Movie Scene Binding Proxy Structure (by ref)", Overwrite:bool=~, ParentBinding:Movie Scene Binding Proxy Structure) {is=title K2Node_FunctionEntry}
   > n2 = Branch(Condition=IsValid!(InBinding:"Movie Scene Binding Proxy Structure (by ref)"=@n1.ParentBinding) -> (ReturnValue))
   > n3 = GetDisplayName!(execute:Exec, InBinding:"Movie Scene Binding Proxy Structure (by ref)"=@n1.ParentBinding) -> (then:Exec, ReturnValue:Text)
   n4 = "To String (Text)"(InText=@n3)
   > Set BindingParentName(BindingParentName:String=@n4) -> (Output_Get:String)
   > n5 = Set IsChild(IsChild:bool=NOT Boolean(A=IsEmpty(InString=@n4))) -> (Output_Get:bool)
   n6 = Sequence!(execute=[@n5.then, @n2.else]) -> (then_0, then_1, then_2:Exec, then_3:Exec)
   n7 = GetTracks!(
    execute=@n6.then_2
    InBinding:"Movie Scene Binding Proxy Structure (by ref)"=Get InBinding() -> (InBinding:"Movie Scene Binding Proxy Structure (by ref)")
   ) -> (then, ReturnValue:Array of Movie Scene Track Object References)
   > n8 = For Each Loop(Array:Array of Movie Scene Track Object References=@n7) -> (Array Element:Movie Scene Track Object Reference)
   > n9 = Cast To MovieSceneFloatTrack!(execute, Object=@n8) -> (then, CastFailed, AsMovie Scene Float Track:Movie Scene Float Track Object Reference) {is=title K2Node_DynamicCast}
   > Add(
    TargetArray:Array of Struct Asset Copier Float Track Structures=Get!() -> (
     Data_BindingOwnerName_31_#1:String
     Data_Overwrite_40_#2:bool=~
     Data_ObjectTemplate_53_#3:Object Reference
     Data_ParentAssetName_21_#4:String
     Data_FloatTracks_22_#5:Array of Struct Asset Copier Float Track Structures
     Data_TransformTrack_43_#6:Struct Asset Copier Transform Track Structure
     Data_BoolTracks_44_#7:Array of Struct Asset Copier Bool Track Structures
     Data_AnimTracks_46_#8:Array of Struct Asset Copier Anim Track Structures
     Data_SpawnTracks_45_#9:Array of Struct Asset Copier Spawn Track Structures
     Data_ControlRigTracks_35_#10:Array of Struct Asset Copier Control Rig Track Structures
     Data_ChildBindings_51_#11:Array of Movie Scene Binding Proxy Structures
     Data_ControlRigBinding_58_#12:Control Rig Sequencer Binding Proxy Structure
    ).Data_FloatTracks_22_#5
    NewItem:"Struct Asset Copier Float Track Structure (by ref)"=GetFloatTrackData!(Track:Movie Scene Track Object Reference=@n9) -> (FloatTrack:Struct Asset Copier Float Track Structure)
   ) {is=title K2Node_CallArrayFunction}
   n10 = Cast To MovieScene3DTransformTrack!(execute=@n9.CastFailed, Object=@n8) -> (then, CastFailed, AsMovie Scene 3DTransform Track:Movie Scene 3DTransform Track Object Reference) {is=title K2Node_DynamicCast}
   n11 = GetTransformTrackData!(Track:Movie Scene Track Object Reference=@n10) -> (TransformSections:Array of Struct Asset Copier Transform Section Structures, PropertyName:Name=None, PropertyPath:String)
   > Set members in Struct Asset Copier Transform Track!(
    execute:Exec
    StructRef:"Struct Asset Copier Transform Track Structure (by ref)"=Get() -> (Data_Overwrite_40_#2=~).Data_TransformTrack_43_#6
    TransformSections_8_#13:Array of Struct Asset Copier Transform Section Structures=@n11
    PropertyName_11_#14:Name=@n11.PropertyName
    PropertyPath_14_#15:String=@n11.PropertyPath
   ) -> (then:Exec, StructOut:"Struct Asset Copier Transform Track Structure (by ref)") {is=title K2Node_SetFieldsInStruct}
   n12 = Cast To MovieSceneBoolTrack!(execute=@n10.CastFailed, Object=@n8) -> (then, CastFailed, AsMovie Scene Bool Track:Movie Scene Bool Track Object Reference) {is=title K2Node_DynamicCast}
   > Add(
    TargetArray:Array of Struct Asset Copier Bool Track Structures=Get() -> (Data_Overwrite_40_#2=~).Data_BoolTracks_44_#7
    NewItem:"Struct Asset Copier Bool Track Structure (by ref)"=GetBoolTrackData!(Track:Movie Scene Track Object Reference=@n12) -> (BoolTrack:Struct Asset Copier Bool Track Structure)
   )
   n13 = Cast To MovieSceneSkeletalAnimationTrack!(execute=@n12.CastFailed, Object=@n8) -> (then, CastFailed, AsMovie Scene Skeletal Animation Track:Movie Scene Skeletal Animation Track Object Reference) {is=title K2Node_DynamicCast}
   > Add(
    TargetArray:Array of Struct Asset Copier Anim Track Structures=Get() -> (Data_Overwrite_40_#2=~).Data_AnimTracks_46_#8
    NewItem:"Struct Asset Copier Anim Track Structure (by ref)"=GetSkeletalAnimData!(Track:Movie Scene Track Object Reference=@n13) -> (AnimTracks:Struct Asset Copier Anim Track Structure)
   )
   n14 = Cast To MovieSceneSpawnTrack!(execute=@n13.CastFailed, Object=@n8) -> (then, CastFailed, AsMovie Scene Spawn Track:Movie Scene Spawn Track Object Reference) {is=title K2Node_DynamicCast}
   > Add!(
    execute
    TargetArray:Array of Struct Asset Copier Spawn Track Structures=Get() -> (Data_Overwrite_40_#2=~).Data_SpawnTracks_45_#9
    NewItem_BoolSection_13_#16:Array of Struct Asset Copier Bool Section Structures=Break Struct Asset Copier Bool Track!(Struct_AssetCopier_BoolTrack:"Struct Asset Copier Bool Track Structure (by ref)"=GetBoolTrackData(Track=@n14)) -> (
     TrackName_8_#17:String
     PropertyName_16_#18:Name
     PropertyPath_17_#19:String
     BoolSections_18_#20:Array of Struct Asset Copier Bool Section Structures
    ) {is=title K2Node_BreakStruct}.BoolSections_18_#20
   ) -> (then, ReturnValue)
   n15 = Cast To MovieSceneControlRigParameterTrack!(execute=@n14.CastFailed, Object=@n8) -> (then, CastFailed, AsMovie Scene Control Rig Parameter Track:Movie Scene Control Rig Parameter Track Object Reference) {is=title K2Node_DynamicCast}
   n16 = GetControlRigData!(Track:Movie Scene Track Object Reference=@n15) -> (ControlRigName:String, CtrlRigSectionsData:Array of Struct Asset Copier Control Rig Section Structures)
   > Add!(
    execute
    TargetArray:Array of Struct Asset Copier Control Rig Track Structures=Get() -> (Data_Overwrite_40_#2=~).Data_ControlRigTracks_35_#10
    NewItem_ControlRigName_2_#21:String=@n16
    NewItem_ControlRigSections_8_#22:Array of Struct Asset Copier Control Rig Section Structures=@n16.CtrlRigSectionsData
   ) -> (then, ReturnValue)
   Return Node!(execute=@n6.then_3, OutData:Struct Asset Copier Binding Data Structure=Get Data() -> (Data:Struct Asset Copier Binding Data Structure))
   n17 = GetBindings!(
    execute=@n6.then_0
    Sequence:Movie Scene Sequence Object Reference=Get Target Level Sequence() -> (Target Level Sequence:Level Sequence Object Reference)
   ) -> (then, ReturnValue:Array of Movie Scene Binding Proxy Structures)
   > n18 = For Each Loop(Array:Array of Movie Scene Binding Proxy Structures=@n17) -> (Array Element:Movie Scene Binding Proxy Structure)
   > n19 = GetDisplayName(InBinding=@n18)
   > Add!(
    execute
    TargetArray:Array of Strings=Get TargetBindingNames() -> (TargetBindingNames:Array of Strings)
    NewItem:"String (by ref)"="To String (Text)"(InText=@n19)
   ) -> (then, ReturnValue)
   n20 = GetDisplayName(execute=@n6.then_1, InBinding=Get InBinding())
   > n21 = GetObjectTemplate!(execute, InBinding:"Movie Scene Binding Proxy Structure (by ref)"=Get InBinding()) -> (then, ReturnValue:Object Reference)
   > n22 = GetChildPossessables!(execute, InBinding:"Movie Scene Binding Proxy Structure (by ref)"=Get InBinding()) -> (then, ReturnValue:Array of Movie Scene Binding Proxy Structures)
   > n23 = Branch(Condition=OR Boolean(A=Get IsChild() -> (IsChild:bool), B=Get Overwrite() -> (Overwrite:bool)))
   n24 = "To String (Text)"(InText=@n20)
   > n25 = Set Name(Name:String=@n24) -> (Output_Get:String)
   Set!(
    execute=[@n27.then, @n25.then]
    Data_BindingOwnerName_31_#1:String=Get Name() -> (Name:String)
    Data_Overwrite_40_#2:bool=Get Overwrite()
    Data_ObjectTemplate_53_#3:Object Reference=@n21
    Data_ParentAssetName_21_#4:String=Get BindingParentName() -> (BindingParentName:String)
    Data_FloatTracks_22_#5:Array of Struct Asset Copier Float Track Structures
    Data_TransformTrack_43_#6:Struct Asset Copier Transform Track Structure=(TransformSections_8_#13=,PropertyName_11_#14="",PropertyPath_14_#15="")
    Data_BoolTracks_44_#7:Array of Struct Asset Copier Bool Track Structures
    Data_AnimTracks_46_#8:Array of Struct Asset Copier Anim Track Structures
    Data_SpawnTracks_45_#9:Array of Struct Asset Copier Spawn Track Structures
    Data_ControlRigTracks_35_#10:Array of Struct Asset Copier Control Rig Track Structures
    Data_ChildBindings_51_#11:Array of Movie Scene Binding Proxy Structures=@n22
    Data_ControlRigBinding_58_#12:Control Rig Sequencer Binding Proxy Structure=GetControlRig!(InBinding:"Movie Scene Binding Proxy Structure (by ref)"=Get InBinding()) -> (ControlRig:Control Rig Sequencer Binding Proxy Structure)
   ) -> (then, Output_Get:Struct Asset Copier Binding Data Structure)
   n26 = UpVersion!(execute=@n23.else, self, InName:String=@n24, NameList:Array of Strings=Get TargetBindingNames()) -> (then, OutName:String)
   > n27 = Set Name(Name=@n26)
  ]}
 CopyAsset {nodes [
   CopyAsset() {is=title K2Node_FunctionEntry}
   > Is Valid(InputObject=Get Source Level Sequence() -> (Source Level Sequence:Level Sequence Object Reference))
   > n1 = GetFocusedLevelSequence!(execute) -> (then, ReturnValue:Level Sequence Object Reference)
   > n2 = Is Valid(InputObject=@n1)
   PrintString(execute=@n2.Is Not Valid, InString=WAITING FOR SEQUENCER TO OPEN)
   Clear and Invalidate Timer by Handle!(execute=@n2.Is Valid, WorldContextObject:Object Reference, Handle=Get Timer() -> (Timer:Timer Handle Structure)) -> (then)
   > n3 = For Each Loop(Array:Array of Widget Object References=GetAllChildren(self=Get SourceScrollBox() -> (SourceScrollBox:Scroll Box Object Reference))) -> (Array Element:Widget Object Reference)
   > n4 = Cast To EUW_LevelSequenceAssetCopier_SourceRow!(execute, Object=@n3) -> (then, CastFailed, AsEUW Level Sequence Asset Copier Source Row:EUW Level Sequence Asset Copier Source Row Object Reference) {is=title K2Node_DynamicCast}
   > Set CurrentRow(CurrentRow:EUW Level Sequence Asset Copier Source Row Object Reference=@n4) -> (Output_Get:EUW Level Sequence Asset Copier Source Row Object Reference)
   > Branch(Condition=IsChecked(self=Get CheckBox!(self:EUW Level Sequence Asset Copier Source Row Object Reference=@n4) -> (CheckBox:Check Box Object Reference)))
   n5 = Get CurrentRow() -> (CurrentRow:EUW Level Sequence Asset Copier Source Row Object Reference)
   > n6 = CollectBindingData!(
    execute
    self
    InBinding:"Movie Scene Binding Proxy Structure (by ref)"=Get In Binding!(self:EUW Level Sequence Asset Copier Source Row Object Reference=@n5) -> (In Binding:Movie Scene Binding Proxy Structure)
    Overwrite:bool=IsChecked(self=Get Overwrite_CheckBox!(self:EUW Level Sequence Asset Copier Source Row Object Reference=@n5) -> (Overwrite_CheckBox:Check Box Object Reference))
    ParentBinding:Movie Scene Binding Proxy Structure
   ) -> (then, OutData:Struct Asset Copier Binding Data Structure) {is=title K2Node_CallFunction}
   > PasteAssets!(execute, self, CurrentBindingData:Struct Asset Copier Binding Data Structure=@n6) -> (then)
  ]}
 Create Rows {nodes [
   n1 = Create Rows!() -> (then, ScrollBox:Scroll Box Object Reference, Sequence:Movie Scene Sequence Object Reference) {is=title K2Node_FunctionEntry}
   > ClearChildren(self=@n1)
   n2 = Get Sequence() -> (Sequence:Movie Scene Sequence Object Reference)
   > Is Valid(InputObject=@n2)
   > n3 = GetBindings(Sequence=@n2)
   > n4 = For Each Loop(Array:Array of Movie Scene Binding Proxy Structures=@n3) -> (Array Element:Movie Scene Binding Proxy Structure)
   > n5 = GetParent!(execute, InBinding:"Movie Scene Binding Proxy Structure (by ref)"=@n4) -> (then, ReturnValue:Movie Scene Binding Proxy Structure)
   > n6 = Branch(Condition=IsValid(InBinding=@n5))
   Set Local Current Binding(execute=@n6.else, Local Current Binding:Movie Scene Binding Proxy Structure=@n4) -> (Output_Get:Movie Scene Binding Proxy Structure)
   > n7 = Create Widget!(
    execute:Exec
    WorldContextObject:Object Reference
    Class:User Widget Class Reference=&/EUW_LevelSequenceAssetCopier_SourceRow.EUW_LevelSequenceAssetCopier_SourceRow_C
    OwningPlayer:Player Controller Object Reference
   ) -> (then:Exec, ReturnValue:EUW Level Sequence Asset Copier Source Row Object Reference)
   > Set Current Row Widget(Current Row Widget:EUW Level Sequence Asset Copier Source Row Object Reference=@n7) -> (Output_Get:EUW Level Sequence Asset Copier Source Row Object Reference)
   > AddChild(
    self=Get ScrollBox() -> (ScrollBox:Scroll Box Object Reference)
    Content=Get Current Row Widget() -> (Current Row Widget:EUW Level Sequence Asset Copier Source Row Object Reference)
   )
   > n8 = GetDisplayName(InBinding=Get Local Current Binding() -> (Local Current Binding:Movie Scene Binding Proxy Structure))
   > Set Text(
    Text=@n8
    self=Get BindingName!(self:EUW Level Sequence Asset Copier Source Row Object Reference=Get Current Row Widget()) -> (BindingName:Text Object Reference)
   )
   n9 = Get Current Row Widget()
   > Set In Binding!(
    execute
    In Binding:Movie Scene Binding Proxy Structure=Get Local Current Binding()
    self:EUW Level Sequence Asset Copier Source Row Object Reference=@n9
   ) -> (then, Output_Get:Movie Scene Binding Proxy Structure)
   > Branch(Condition="Equal (Object)"(A=Get Sequence(), B=Get Target Level Sequence()))
   > Set Visibility(Visibility=Hidden, self=Get CheckBox(self=@n9))
   > Set Visibility(Visibility=Hidden, self=Get Overwrite_CheckBox(self=@n9))
   > Set Visibility(
    Visibility=Visible
    self=Get TargetBindingName!(self:EUW Level Sequence Asset Copier Source Row Object Reference=@n9) -> (TargetBindingName:Text Object Reference)
   )
  ]}
 CreateAnimTrack {nodes [
   n1 = CreateAnimTrack!() -> (
    then
    InBinding:"Movie Scene Binding Proxy Structure (by ref)"
    TrackName:String
    TrackSections:Array of Struct Asset Copier Anim Section Structures
   ) {is=title K2Node_FunctionEntry}
   > n2 = AddTrack!(
    execute
    InBinding:"Movie Scene Binding Proxy Structure (by ref)"=@n1
    TrackType:Movie Scene Track Class Reference=/Script/MovieSceneTracks.MovieSceneSkeletalAnimationTrack
   ) -> (then, ReturnValue:Movie Scene Track Object Reference)
   > n3 = Cast To MovieSceneSkeletalAnimationTrack(Object=@n2)
   > SetDisplayName!(
    execute
    Track:Movie Scene Track Object Reference=@n3
    InName:"Text (by ref)"="To Text (String)"(InString=Get TrackName() -> (TrackName:String))
   ) -> (then)
   > n4 = For Each Loop!(
    Exec
    Array:Array of Struct Asset Copier Anim Section Structures=Get TrackSections() -> (TrackSections:Array of Struct Asset Copier Anim Section Structures)
   ) -> (
    LoopBody
    Array Element_SectionStartFrames_11_#23:Array of Integers
    Array Element_SectionEndFrames_12_#24:Array of Integers
    Array Element_AnimParams_27_#25:Array of Movie Scene Skeletal Animation Params Structures
    Array Element_RowIndex_33_#26:Array of Integers
    Array Index
    Completed
   )
   > n5 = For Each Loop!(Exec, Array:Array of Integers=@n4) -> (LoopBody, Array Element:int, Array Index, Completed)
   > n6 = AddSection(Track=@n3)
   > SetStartFrame!(execute, Section:Movie Scene Section Object Reference=@n6, StartFrame:int=@n5) -> (then)
   n7 = "Get (a ref)"(Array=@n4.Array Element_SectionEndFrames_12_#24, Dimension 1=@n5.Array Index)
   > SetEndFrame!(execute, Section:Movie Scene Section Object Reference=@n6, EndFrame:int=@n7) -> (then)
   > n8 = Cast To MovieSceneSkeletalAnimationSection!(execute, Object=@n6) -> (then, CastFailed, AsMovie Scene Skeletal Animation Section:Movie Scene Skeletal Animation Section Object Reference) {is=title K2Node_DynamicCast}
   > SetRowIndex(self=@n8, NewRowIndex="Get (a ref)"(Array=@n4.Array Element_RowIndex_33_#26, Dimension 1=@n5.Array Index))
   > Set Params!(
    execute
    Params:Movie Scene Skeletal Animation Params Structure="Get (a ref)"(
     Array:Array of Movie Scene Skeletal Animation Params Structures=@n4.Array Element_AnimParams_27_#25
     Dimension 1=@n5.Array Index
    ) -> (Output:"Movie Scene Skeletal Animation Params Structure (by ref)")
    self:Movie Scene Skeletal Animation Section Object Reference=@n8
   ) -> (then, Output_Get:Movie Scene Skeletal Animation Params Structure)
   > SetStartFrame(Section=@n8, StartFrame=@n5)
   > SetEndFrame(Section=@n8, EndFrame=@n7)
  ]}
 CreateBoolTrack {nodes [
   n1 = CreateBoolTrack!() -> (
    then
    TrackSections:Array of Struct Asset Copier Bool Section Structures
    TrackName:String
    InBinding:"Movie Scene Binding Proxy Structure (by ref)"
    PropertyName:Name
    PropertyPath:String
   ) {is=title K2Node_FunctionEntry}
   > n2 = AddTrack(InBinding=@n1.InBinding, TrackType=/Script/MovieSceneTracks.MovieSceneFloatTrack)
   > n3 = Cast To MovieSceneBoolTrack(Object=@n2)
   > SetDisplayName(Track=@n3, InName="To Text (String)"(InString=Get TrackName()))
   > SetPropertyNameAndPath!(
    execute
    Track:Movie Scene Property Track Object Reference=@n3
    InPropertyName:"Name (by ref)"=Get PropertyName() -> (PropertyName:Name=None)
    InPropertyPath:String=Get PropertyPath() -> (PropertyPath:String)
   ) -> (then)
   > n4 = For Each Loop!(
    Exec
    Array:Array of Struct Asset Copier Bool Section Structures=Get TrackSections() -> (TrackSections:Array of Struct Asset Copier Bool Section Structures)
   ) -> (
    LoopBody
    Array Element_BoolChannel_5_#27:Array of Struct Asset Copier Bool Channel Structures
    Array Index
    Completed
   )
   > AddBoolSection!(execute, self, Track:Movie Scene Track Object Reference=@n3, BoolChannels:Array of Struct Asset Copier Bool Channel Structures=@n4) -> (then) {is=title K2Node_CallFunction}
  ]}
 CreateCtrlRigTrack {nodes [
   n1 = CreateCtrlRigTrack!() -> (
    then
    InBinding:"Movie Scene Binding Proxy Structure (by ref)"
    CtrlRigSections:Array of Struct Asset Copier Control Rig Section Structures
    TrackName:String
   ) {is=title K2Node_FunctionEntry}
   > n2 = AddTrack(InBinding=@n1, TrackType=/Script/ControlRig.MovieSceneControlRigParameterTrack)
   > SetDisplayName(Track=@n2, InName="To Text (String)"(InString=Get TrackName()))
   > n3 = For Each Loop!(
    Exec
    Array:Array of Struct Asset Copier Control Rig Section Structures=Get CtrlRigSections() -> (CtrlRigSections:Array of Struct Asset Copier Control Rig Section Structures)
   ) -> (
    LoopBody
    Array Element_BlendType_20_#28:EMovieSceneBlendType Enum
    Array Element_RowDisplayName_30_#29:Text
    Array Element_FloatControls_13_#30:Array of Struct Asset Copier Float Channel Structures
    Array Element_BoolControls_14_#31:Array of Struct Asset Copier Bool Channel Structures
    Array Index
    Completed
   )
   > AddCtrlRigSection!(
    execute
    self
    Track:Movie Scene Track Object Reference=@n2
    InBlendType:EMovieSceneBlendType Enum=@n3
    TrackRowName:Text=@n3.Array Element_RowDisplayName_30_#29
    FloatChannels:Array of Struct Asset Copier Float Channel Structures=@n3.Array Element_FloatControls_13_#30
    BoolChannels:Array of Struct Asset Copier Bool Channel Structures=@n3.Array Element_BoolControls_14_#31
    SectionRowIndex:int=@n3.Array Index
   ) -> (then) {is=title K2Node_CallFunction}
  ]}
 CreateFloatTrack {nodes [
   n1 = CreateFloatTrack!() -> (
    then
    TrackSections:Array of Struct Asset Copier Float Section Structures
    TrackName:String
    InBinding:"Movie Scene Binding Proxy Structure (by ref)"
    PropertyName:"Name (by ref)"
    PropertyPath:String
   ) {is=title K2Node_FunctionEntry}
   > n2 = AddTrack(InBinding=@n1.InBinding, TrackType=&MovieSceneFloatTrack)
   > n3 = Cast To MovieSceneFloatTrack(Object=@n2)
   > SetDisplayName(Track=@n3, InName="To Text (String)"(InString=Get TrackName()))
   > SetPropertyNameAndPath(Track=@n3, InPropertyName=Get PropertyName() -> (PropertyName:"Name (by ref)"=None), InPropertyPath=Get PropertyPath())
   > n4 = For Each Loop!(
    Exec
    Array:Array of Struct Asset Copier Float Section Structures=Get TrackSections() -> (TrackSections:Array of Struct Asset Copier Float Section Structures)
   ) -> (
    LoopBody
    Array Element_BlendType_5_#32:EMovieSceneBlendType Enum
    Array Element_FloatData_10_#33:Array of Struct Asset Copier Float Channel Structures
    Array Index
    Completed
   )
   > AddFloatSection!(
    execute
    self
    Track:Movie Scene Track Object Reference=@n3
    InBlendType:EMovieSceneBlendType Enum=@n4
    FloatChannels:Array of Struct Asset Copier Float Channel Structures=@n4.Array Element_FloatData_10_#33
    SectionRowIndex:int=@n4.Array Index
   ) -> (then) {is=title K2Node_CallFunction}
  ]}
 CreateSpawnTrack {nodes [
   n1 = CreateSpawnTrack!() -> (then, SectionChannels:Array of Struct Asset Copier Bool Channel Structures, InBinding:"Movie Scene Binding Proxy Structure (by ref)") {is=title K2Node_FunctionEntry}
   > n2 = AddTrack(InBinding=@n1.InBinding, TrackType=/Script/MovieScene.MovieSceneSpawnTrack)
   > n3 = Cast To MovieSceneSpawnTrack(Object=@n2)
   > addSpawnedSection!(execute, self, Track:Movie Scene Track Object Reference=@n3, BoolChannels:Array of Struct Asset Copier Bool Channel Structures) -> (then)
  ]}
 CreateTransformTrack {nodes [
   n1 = CreateTransformTrack!() -> (
    then
    InBinding:"Movie Scene Binding Proxy Structure (by ref)"
    TransformData:Array of Struct Asset Copier Transform Section Structures
    PropertyName:Name
    PropertyPath:String
   ) {is=title K2Node_FunctionEntry}
   > n2 = AddTrack(InBinding=@n1, TrackType=/Script/MovieSceneTracks.MovieScene3DTransformTrack)
   > n3 = Cast To MovieScene3DTransformTrack(Object=@n2)
   > SetPropertyNameAndPath(Track=@n3, InPropertyName=Get PropertyName() -> (PropertyName:Name=None), InPropertyPath=Get PropertyPath())
   > n4 = For Each Loop!(
    Exec
    Array:Array of Struct Asset Copier Transform Section Structures=Get TransformData() -> (TransformData:Array of Struct Asset Copier Transform Section Structures)
   ) -> (
    LoopBody
    Array Element_BlendType_8_#34:EMovieSceneBlendType Enum
    Array Element_FloatData_12_#35:Array of Struct Asset Copier Float Channel Structures
    Array Index
    Completed
   )
   > AddTransformSection!(
    execute
    self
    InBlendType:EMovieSceneBlendType Enum=@n4
    FloatData:Array of Struct Asset Copier Float Channel Structures=@n4.Array Element_FloatData_12_#35
    SectionRowIndex:int=@n4.Array Index
    Track:Movie Scene Track Object Reference=@n3
   ) -> (then) {is=title K2Node_CallFunction}
  ]}
 EventGraph {nodes [
   n1 = "On Property Changed (SourceLevelSequencePropertyView)"!() -> (OutputDelegate, then, PropertyName:Name=None) {is=title K2Node_ComponentBoundEvent}
   ClearChildren(execute=[@n1.then, @n9.then], self=Get SourceScrollBox())
   > n2 = LoadAsset(AssetPath=Get Object Path String(Object=Get Source Level Sequence()))
   > n3 = Cast To LevelSequence!(execute, Object=@n2) -> (then, CastFailed, AsLevel Sequence:Level Sequence Object Reference) {is=title K2Node_DynamicCast}
   n4 = Get SourceScrollBox()
   > Create Rows!(execute, self, ScrollBox:Scroll Box Object Reference=@n4, Sequence:Movie Scene Sequence Object Reference=@n3) -> (then) {is=title K2Node_CallFunction}
   > Sort Scroll Box By Checkbox!(execute, self, ScrollBox:Panel Widget Object Reference=@n4) -> (then)
   > PreLoadSource()
   "On Clicked (CopyButton)"() {is=title K2Node_ComponentBoundEvent}
   n5 = Get Target Level Sequence()
   > Is Valid(InputObject=@n5)
   > OpenLevelSequence!(execute, LevelSequence:Level Sequence Object Reference=@n5) -> (then, ReturnValue:bool)
   > n6 = Set Timer by Function Name!(
    execute
    Object:Object Reference
    FunctionName:String=CopyAsset
    Time:float=0.2
    bLooping:bool=true
    bMaxOncePerFrame:bool=true
    InitialStartDelay:float=0
    InitialStartDelayVariance:float=0
   ) -> (then, ReturnValue:Timer Handle Structure) {is=title K2Node_CallFunction}
   > Set Timer(Timer:Timer Handle Structure=@n6) -> (Output_Get:Timer Handle Structure)
   n7 = "On Clicked (CurrentLevelSequenceButton)"() {is=title K2Node_ComponentBoundEvent}
   n8 = GetFocusedLevelSequence(execute=[@n7.then, @n12.then])
   > n9 = Set Source Level Sequence(Source Level Sequence:Level Sequence Object Reference=@n8) -> (Output_Get:Level Sequence Object Reference)
   "On Property Changed (TargetLevelSequencePropertyView)"!() -> (OutputDelegate, then, PropertyName:Name=None) {is=title K2Node_ComponentBoundEvent}
   n10 = Get TargetScrollBox() -> (TargetScrollBox:Scroll Box Object Reference)
   > Create Rows(ScrollBox=@n10, Sequence=Get Target Level Sequence())
   > Sort Scroll Box By Checkbox(ScrollBox=@n10)
   Event PreConstruct()
   n11 = Self-Reference()
   > SetObject(self=Get SourceLevelSequencePropertyView() -> (SourceLevelSequencePropertyView:Single Property View Object Reference), NewObject=@n11)
   > SetObject(self=Get TargetLevelSequencePropertyView() -> (TargetLevelSequencePropertyView:Single Property View Object Reference), NewObject=@n11)
   n12 = Event Construct()
  ]}
 FrameRateConversion {nodes [
   FrameRateConversion!() -> (then, FrameNumber:"Frame Number Structure (by ref)") {is=title K2Node_FunctionEntry}
   > n1 = GetDisplayRate!(execute, Sequence:Movie Scene Sequence Object Reference=Get Source Level Sequence()) -> (then, ReturnValue_Numerator:int=0, ReturnValue_Denominator:int=0)
   > n2 = GetDisplayRate(Sequence=Get Target Level Sequence()) -> (ReturnValue_Numerator=0, ReturnValue_Denominator=0)
   n3 = Get FrameNumber() -> (FrameNumber:"Frame Number Structure (by ref)")
   n4 = float / float(
    A="To Float (Integer)"(InInt=Break Frame Number!(FrameNumber:"Frame Number Structure (by ref)"=@n3) -> (Value:int) {is=title K2Node_BreakStruct})
    B=float / float(A="To Float (Integer)"(InInt=@n1), B="To Float (Integer)"(InInt=@n2))
   )
   > Return Node!(
    execute
    ConvertedFrameNumber:Frame Number Structure=Make Frame Number!(Value:int=Truncate(A=@n4)) -> (FrameNumber:Frame Number Structure) {is=title K2Node_MakeStruct}
   )
  ]}
 GenerateNewBinding {nodes [
   GenerateNewBinding!() -> (then, CurrentSourceBinding:Struct Asset Copier Binding Data Structure) {is=title K2Node_FunctionEntry}
   n1 = Get!() -> (
    CurrentSourceBinding_BindingOwnerName_31_#1:String
    CurrentSourceBinding_Overwrite_40_#2:bool=~
    CurrentSourceBinding_ObjectTemplate_53_#3:Object Reference
    CurrentSourceBinding_ParentAssetName_21_#4:String
    CurrentSourceBinding_FloatTracks_22_#5:Array of Struct Asset Copier Float Track Structures
    CurrentSourceBinding_TransformTrack_43_#6:Struct Asset Copier Transform Track Structure
    CurrentSourceBinding_BoolTracks_44_#7:Array of Struct Asset Copier Bool Track Structures
    CurrentSourceBinding_AnimTracks_46_#8:Array of Struct Asset Copier Anim Track Structures
    CurrentSourceBinding_SpawnTracks_45_#9:Array of Struct Asset Copier Spawn Track Structures
    CurrentSourceBinding_ControlRigTracks_35_#10:Array of Struct Asset Copier Control Rig Track Structures
    CurrentSourceBinding_ChildBindings_51_#11:Array of Movie Scene Binding Proxy Structures
    CurrentSourceBinding_ControlRigBinding_58_#12:Control Rig Sequencer Binding Proxy Structure
   )
   > Branch(Condition=IsEmpty(InString=@n1.CurrentSourceBinding_ParentAssetName_21_#4))
   > n2 = AddSpawnableFromInstance!(
    execute
    self:Level Sequence Editor Subsystem Object Reference=Get LevelSequenceEditorSubsystem() -> (ReturnValue:Level Sequence Editor Subsystem Object Reference) {is=title K2Node_GetEditorSubsystem}
    Sequence:Movie Scene Sequence Object Reference=Get Target Level Sequence()
    ObjectToSpawn:Object Reference=@n1.CurrentSourceBinding_ObjectTemplate_53_#3
   ) -> (then, ReturnValue:Movie Scene Binding Proxy Structure)
   > SetDisplayName!(execute, InBinding:"Movie Scene Binding Proxy Structure (by ref)"=@n2, InDisplayName:"Text (by ref)"="To Text (String)"(InString=@n1)) -> (then)
   > n3 = GetEditorWorld(self=Get UnrealEditorSubsystem())
   > FindOrCreateControlRigTrackFromReference!(
    execute
    World:World Object Reference=@n3
    LevelSequence:Level Sequence Object Reference=Get Target Level Sequence()
    InRef:"Control Rig Asset Strong Reference Structure (by ref)"=GetAssetReference!(self:Control Rig Object Reference=Break Control Rig Sequencer Binding Proxy!(ControlRigSequencerBindingProxy:"Control Rig Sequencer Binding Proxy Structure (by ref)"=@n1.CurrentSourceBinding_ControlRigBinding_58_#12) -> (
      Proxy:Movie Scene Binding Proxy Structure
      ControlRig:Control Rig Object Reference
      Track:Movie Scene Control Rig Parameter Track Object Reference
     ) {is=title K2Node_BreakStruct}.ControlRig) -> (ReturnValue:Control Rig Asset Strong Reference Structure)
    InBinding:"Movie Scene Binding Proxy Structure (by ref)"=@n2
    bIsLayeredControlRig:bool=false
   ) -> (then, ReturnValue:Movie Scene Track Object Reference)
   > n4 = OverwriteBinding!(
    execute
    self
    CurrentSourceBinding:Struct Asset Copier Binding Data Structure=Get CurrentSourceBinding() -> (CurrentSourceBinding:Struct Asset Copier Binding Data Structure)
   ) -> (then, OutTargetBinding:Movie Scene Binding Proxy Structure)
   > Return Node!(execute, OutTargetBinding:Movie Scene Binding Proxy Structure=@n4)
  ]}
 GetBoolChannelData {nodes [
   n1 = GetBoolChannelData!() -> (then, Object:Object Reference) {is=title K2Node_FunctionEntry}
   > n2 = Cast To MovieSceneScriptingBoolChannel!(execute, Object=@n1) -> (then, CastFailed, AsMovie Scene Scripting Bool Channel:Movie Scene Scripting Bool Channel Object Reference) {is=title K2Node_DynamicCast}
   n3 = "Get Keys (Bool)"!(self:Movie Scene Scripting Bool Channel Object Reference=@n2) -> (ReturnValue:Array of Movie Scene Scripting Key Object References) {is=title K2Node_CallFunction}
   > n4 = Branch(Condition="integer > integer"(A=Length(TargetArray:Array of Movie Scene Scripting Key Object References=@n3)))
   > n5 = For Each Loop!(Exec, Array:Array of Movie Scene Scripting Key Object References=@n3) -> (LoopBody, Array Element:Movie Scene Scripting Key Object Reference, Array Index, Completed)
   n6 = GetDisplayName!(Object:Object Reference=@n2) -> (ReturnValue:String)
   n7 = FindSubstring(SearchIn=@n6, Substring=_)
   Return Node!(
    execute=[@n9.then, @n5.Completed]
    Bool Channels:Struct Asset Copier Bool Channel Structure=Make Struct Asset Copier Bool Channel!(
     OwnerName_19_#36:String=Left!(SourceString:String=@n6, Count:int=@n7) -> (ReturnValue:String)
     FrameTime_14_#37:Array of Frame Time Structures=Get FrameTimes() -> (FrameTimes:Array of Frame Time Structures)
     FrameValue_7_#38:Array of Booleans=Get BoolArray() -> (BoolArray:Array of Booleans)
    ) -> (Struct_AssetCopier_BoolChannel:Struct Asset Copier Bool Channel Structure) {is=title K2Node_MakeStruct}
   )
   > n8 = Cast To MovieSceneScriptingBoolKey!(execute, Object=@n5) -> (then, CastFailed, AsMovie Scene Scripting Bool Key:Movie Scene Scripting Bool Key Object Reference) {is=title K2Node_DynamicCast}
   > Add(
    TargetArray:Array of Frame Time Structures=Get FrameTimes()
    NewItem:"Frame Time Structure (by ref)"="Get Time (Bool)"!(self:Movie Scene Scripting Bool Key Object Reference=@n8, TimeUnit:EMovieSceneTimeUnit Enum=DisplayRate) -> (ReturnValue:Frame Time Structure) {is=title K2Node_CallFunction}
   )
   > Add(
    TargetArray:Array of Booleans=Get BoolArray()
    NewItem:"Boolean (by ref)"="Get Value (Bool)"!(self:Movie Scene Scripting Bool Key Object Reference=@n8) -> (ReturnValue:bool) {is=title K2Node_CallFunction}
   )
   n9 = Add(
    execute=@n4.else
    TargetArray=Get BoolArray()
    NewItem="Get Default (Bool)"!(self:Movie Scene Scripting Bool Channel Object Reference=@n2) -> (ReturnValue:bool) {is=title K2Node_CallFunction}
   )
  ]}
 GetBoolTrackData {nodes [
   n1 = GetBoolTrackData!() -> (then, Track:Movie Scene Track Object Reference) {is=title K2Node_FunctionEntry}
   > n2 = GetSections!(execute, Track:Movie Scene Track Object Reference=@n1) -> (then, ReturnValue:Array of Movie Scene Section Object References)
   > n3 = For Each Loop(Array:Array of Movie Scene Section Object References=@n2) -> (Array Element:Movie Scene Section Object Reference)
   n4 = Get Track() -> (Track:Movie Scene Track Object Reference)
   n5 = GetDisplayName!(execute:Exec=@n3.Completed, Track:Movie Scene Track Object Reference=@n4) -> (then:Exec, ReturnValue:Text)
   > n6 = Cast To MovieSceneBoolTrack(Object=@n4)
   > n7 = GetPropertyName!(execute, Track:Movie Scene Property Track Object Reference=@n6) -> (then, ReturnValue:Name=None)
   > n8 = GetPropertyPath!(execute, Track:Movie Scene Property Track Object Reference=@n6) -> (then, ReturnValue:String)
   Return Node!(
    execute=[@n8.then, @n9.then]
    BoolTrack_TrackName_8_#17:String="To String (Text)"(InText=@n5)
    BoolTrack_PropertyName_16_#18:Name=@n7
    BoolTrack_PropertyPath_17_#19:String=@n8
    BoolTrack_BoolSections_18_#20:Array of Struct Asset Copier Bool Section Structures=Get BoolSections() -> (BoolSections:Array of Struct Asset Copier Bool Section Structures)
   )
   n9 = Cast To MovieSceneSpawnTrack(execute=@n6.CastFailed, Object=@n4)
   n10 = GetChannelsByType!(
    execute=@n3.LoopBody
    Section:Movie Scene Section Object Reference=@n3
    ChannelType:Movie Scene Scripting Channel Class Reference=/Script/SequencerScripting.MovieSceneScriptingBoolChannel
   ) -> (then, ReturnValue:Array of Movie Scene Scripting Channel Object References)
   > n11 = For Each Loop(Array:Array of Movie Scene Scripting Channel Object References=@n10) -> (Array Element:Movie Scene Scripting Channel Object Reference)
   Add!(
    execute=@n11.Completed
    TargetArray:Array of Struct Asset Copier Bool Section Structures=Get BoolSections()
    NewItem_BoolChannel_5_#27:Array of Struct Asset Copier Bool Channel Structures=Get BoolChannels()
   ) -> (then, ReturnValue)
   > Clear(TargetArray:Array of Struct Asset Copier Bool Channel Structures=Get BoolChannels())
   n12 = GetBoolChannelData!(execute=@n11.LoopBody, self, Object:Object Reference=@n11) -> (then, Bool Channels:Struct Asset Copier Bool Channel Structure) {is=title K2Node_CallFunction}
   > Add!(
    execute
    TargetArray:Array of Struct Asset Copier Bool Channel Structures=Get BoolChannels()
    NewItem:"Struct Asset Copier Bool Channel Structure (by ref)"=@n12
   ) -> (then, ReturnValue)
  ]}
 GetControlRig {nodes [
   GetControlRig!() -> (then, InBinding:"Movie Scene Binding Proxy Structure (by ref)") {is=title K2Node_FunctionEntry}
   > n1 = GetControlRigs!(execute, LevelSequence:Level Sequence Object Reference=Get Source Level Sequence()) -> (then, ReturnValue:Array of Control Rig Sequencer Binding Proxy Structures)
   > n2 = For Each Loop(Array:Array of Control Rig Sequencer Binding Proxy Structures=@n1) -> (Array Element:Control Rig Sequencer Binding Proxy Structure)
   Return Node!(
    execute=@n2.Completed
    ControlRig:Control Rig Sequencer Binding Proxy Structure=Get rig() -> (rig:Control Rig Sequencer Binding Proxy Structure)
   )
   > Branch(Condition="Equal (MovieSceneBindingProxy)"(
     A:Movie Scene Binding Proxy Structure=Break Control Rig Sequencer Binding Proxy(ControlRigSequencerBindingProxy=@n2)
     B:Movie Scene Binding Proxy Structure=Get InBinding()
    ) -> (ReturnValue:bool=~) {is=title K2Node_PromotableOperator})
   > Set rig(rig:Control Rig Sequencer Binding Proxy Structure=@n2) -> (Output_Get:Control Rig Sequencer Binding Proxy Structure)
  ]}
 GetControlRigData {nodes [
   n1 = GetControlRigData!() -> (then, Track:Movie Scene Track Object Reference) {is=title K2Node_FunctionEntry}
   > n2 = GetSections(Track=@n1)
   > n3 = For Each Loop(Array:Array of Movie Scene Section Object References=@n2) -> (Array Element:Movie Scene Section Object Reference)
   n4 = GetDisplayName(execute=@n3.Completed, Track=Get Track())
   > Return Node!(
    execute
    ControlRigName:String="To String (Text)"(InText=@n4)
    CtrlRigSectionsData:Array of Struct Asset Copier Control Rig Section Structures=Get SectionData() -> (SectionData:Array of Struct Asset Copier Control Rig Section Structures)
   )
   n5 = Cast To MovieSceneControlRigParameterSection!(execute=@n3.LoopBody, Object=@n3) -> (then, CastFailed, AsMovie Scene Control Rig Parameter Section:Movie Scene Control Rig Parameter Section Object Reference) {is=title K2Node_DynamicCast}
   > n6 = GetAllChannels!(execute, Section:Movie Scene Section Object Reference=@n5) -> (then, ReturnValue:Array of Movie Scene Scripting Channel Object References)
   > n7 = For Each Loop(Array:Array of Movie Scene Scripting Channel Object References=@n6) -> (Array Element:Movie Scene Scripting Channel Object Reference)
   n8 = GetTrackRowDisplayName!(execute=@n7.Completed, Track:Movie Scene Track Object Reference=Get Track(), RowIndex:int=@n3.Array Index) -> (then, ReturnValue:Text)
   > Add!(
    execute
    TargetArray:Array of Struct Asset Copier Control Rig Section Structures=Get SectionData()
    NewItem_BlendType_20_#28:EMovieSceneBlendType Enum=GetBlendType!(self:Movie Scene Section Object Reference=@n3) -> (ReturnValue_BlendType:EMovieSceneBlendType Enum, ReturnValue_bIsValid:bool=~)
    NewItem_RowDisplayName_30_#29:Text=@n8
    NewItem_FloatControls_13_#30:Array of Struct Asset Copier Float Channel Structures=Get FloatData()
    NewItem_BoolControls_14_#31:Array of Struct Asset Copier Bool Channel Structures=Get BoolData() -> (BoolData:Array of Struct Asset Copier Bool Channel Structures)
   ) -> (then, ReturnValue)
   > Clear(TargetArray:Array of Struct Asset Copier Float Channel Structures=Get FloatData())
   > Clear(TargetArray:Array of Struct Asset Copier Bool Channel Structures=Get BoolData())
   n9 = GetFloatChannelData!(execute=@n7.LoopBody, self, Channel:Movie Scene Scripting Channel Object Reference=@n7) -> (then, FloatData:Struct Asset Copier Float Channel Structure)
   > Add!(
    execute
    TargetArray:Array of Struct Asset Copier Float Channel Structures=Get FloatData()
    NewItem:"Struct Asset Copier Float Channel Structure (by ref)"=@n9
   ) -> (then, ReturnValue)
   > n10 = GetBoolChannelData(Object=@n7)
   > Add(
    TargetArray:Array of Struct Asset Copier Bool Channel Structures=Get BoolData()
    NewItem:"Struct Asset Copier Bool Channel Structure (by ref)"=@n10
   )
  ]}
 GetFloatChannelData {nodes [
   n1 = GetFloatChannelData!() -> (then, Channel:Movie Scene Scripting Channel Object Reference) {is=title K2Node_FunctionEntry}
   > n2 = Cast To MovieSceneScriptingFloatChannel!(execute, Object=@n1) -> (then, CastFailed, AsMovie Scene Scripting Float Channel:Movie Scene Scripting Float Channel Object Reference) {is=title K2Node_DynamicCast}
   n3 = "Get Keys (Float)"!(self:Movie Scene Scripting Float Channel Object Reference=@n2) -> (ReturnValue:Array of Movie Scene Scripting Key Object References) {is=title K2Node_CallFunction}
   > n4 = Branch(Condition="integer > integer"(A=Length(TargetArray=@n3)))
   > n5 = For Each Loop(Array:Array of Movie Scene Scripting Key Object References=@n3) -> (Array Element:Movie Scene Scripting Key Object Reference)
   n6 = GetDisplayName!(Object:Object Reference=@n2) -> (ReturnValue:String)
   n7 = FindSubstring(SearchIn=@n6, Substring=_)
   Return Node!(
    execute=[@n9.then, @n5.Completed]
    FloatData:Struct Asset Copier Float Channel Structure=Make Struct Asset Copier Float Channel!(
     OwnerName_19_#39:String=Left(SourceString=@n6, Count=@n7)
     FrameTime_27_#40:Array of Frame Time Structures=Get FrameTimes()
     FrameValue_33_#41:"Array of Float (double-precision)s"=Get FrameValues() -> (FrameValues:"Array of Float (double-precision)s")
     FloatKeyParams_32_#42:Array of Struct Asset Copier Float Key Params Structures=Get FloatKeyParams() -> (FloatKeyParams:Array of Struct Asset Copier Float Key Params Structures)
    ) -> (Struct_AssetCopier_FloatChannel:Struct Asset Copier Float Channel Structure) {is=title K2Node_MakeStruct}
   )
   > n8 = Cast To MovieSceneScriptingActualFloatKey!(execute, Object=@n5) -> (then, CastFailed, AsMovie Scene Scripting Actual Float Key:Movie Scene Scripting Actual Float Key Object Reference) {is=title K2Node_DynamicCast}
   > Add(
    TargetArray:Array of Frame Time Structures=Get FrameTimes()
    NewItem:"Frame Time Structure (by ref)"="Get Time (Float)"!(self:Movie Scene Scripting Float Key Object Reference=@n8, TimeUnit:EMovieSceneTimeUnit Enum=DisplayRate) -> (ReturnValue:Frame Time Structure) {is=title K2Node_CallFunction}
   )
   > Add(
    TargetArray:"Array of Float (double-precision)s"=Get FrameValues()
    NewItem:"Float (double-precision) (by ref)"="Get Value (Float)"!(self:Movie Scene Scripting Float Key Object Reference=@n8) -> (ReturnValue:float) {is=title K2Node_CallFunction}
   )
   > Add(
    TargetArray:Array of Struct Asset Copier Float Key Params Structures=Get FloatKeyParams()
    NewItem:"Struct Asset Copier Float Key Params Structure (by ref)"=Make Struct Asset Copier Float Key Params!(
     InterpolationMode_2_#43:ERichCurveInterpMode Enum=GetInterpolationMode!(self:Movie Scene Scripting Float Key Object Reference=@n8) -> (ReturnValue:ERichCurveInterpMode Enum=RCIM_Linear)
     TangentMode_18_#44:ERichCurveTangentMode Enum=GetTangentMode!(self:Movie Scene Scripting Float Key Object Reference=@n8) -> (ReturnValue:ERichCurveTangentMode Enum=RCTM_Auto)
     TangentWeightMode_14_#45:ERichCurveTangentWeightMode Enum=GetTangentWeightMode!(self:Movie Scene Scripting Float Key Object Reference=@n8) -> (ReturnValue:ERichCurveTangentWeightMode Enum=RCTWM_WeightedNone)
     ArriveTangent_7_#46:double=GetArriveTangent!(self:Movie Scene Scripting Float Key Object Reference=@n8) -> (ReturnValue:float)
     ArriveTangentWeight_9_#47:double=GetArriveTangentWeight!(self:Movie Scene Scripting Float Key Object Reference=@n8) -> (ReturnValue:float)
     LeaveTangent_13_#48:double=GetLeaveTangent!(self:Movie Scene Scripting Float Key Object Reference=@n8) -> (ReturnValue:float)
     LeaveTangentWeight_17_#49:double=GetLeaveTangentWeight!(self:Movie Scene Scripting Float Key Object Reference=@n8) -> (ReturnValue:float)
    ) -> (Struct_AssetCopier_FloatKeyParams:Struct Asset Copier Float Key Params Structure) {is=title K2Node_MakeStruct}
   )
   n9 = Add(
    execute=@n4.else
    TargetArray:"Array of Float (double-precision)s"=Get FrameValues()
    NewItem:"Float (double-precision) (by ref)"="Get Default (Float)"!(self:Movie Scene Scripting Float Channel Object Reference=@n2) -> (ReturnValue:float) {is=title K2Node_CallFunction}
   )
  ]}
 GetFloatTrackData {nodes [
   n1 = GetFloatTrackData!() -> (then, Track:Movie Scene Track Object Reference) {is=title K2Node_FunctionEntry}
   > n2 = GetSections(Track=@n1)
   > n3 = For Each Loop(Array:Array of Movie Scene Section Object References=@n2) -> (Array Element:Movie Scene Section Object Reference)
   n4 = Get Track()
   n5 = GetDisplayName!(execute:Exec=@n3.Completed, Track:Movie Scene Track Object Reference=@n4) -> (then:Exec, ReturnValue:Text)
   > n6 = Cast To MovieSceneFloatTrack(Object=@n4)
   > n7 = GetPropertyName(Track=@n6) -> (ReturnValue=None)
   > n8 = GetPropertyPath(Track=@n6)
   > Return Node!(
    execute
    FloatTrack_DisplayName_18_#50:String="To String (Text)"(InText=@n5)
    FloatTrack_PropertyName_22_#51:Name=@n7
    FloatTrack_PropertyPath_24_#52:String=@n8
    FloatTrack_FloatSections_13_#53:Array of Struct Asset Copier Float Section Structures=Get Float Section Data() -> (Float Section Data:Array of Struct Asset Copier Float Section Structures)
   )
   n9 = GetChannelsByType(execute=@n3.LoopBody, Section=@n3, ChannelType=/Script/SequencerScripting.MovieSceneScriptingFloatChannel)
   > n10 = For Each Loop(Array:Array of Movie Scene Scripting Channel Object References=@n9) -> (Array Element:Movie Scene Scripting Channel Object Reference)
   Add!(
    execute=@n10.Completed
    TargetArray:Array of Struct Asset Copier Float Section Structures=Get Float Section Data()
    NewItem_BlendType_5_#32:EMovieSceneBlendType Enum=GetBlendType(self=@n3) -> (ReturnValue_bIsValid=~)
    NewItem_FloatData_10_#33:Array of Struct Asset Copier Float Channel Structures=Get Float Data() -> (Float Data:Array of Struct Asset Copier Float Channel Structures)
   ) -> (then, ReturnValue)
   > Clear(TargetArray:Array of Struct Asset Copier Float Channel Structures=Get Float Data())
   n11 = GetFloatChannelData(execute=@n10.LoopBody, Channel=@n10) {is=title K2Node_CallFunction}
   > Add!(
    execute
    TargetArray:Array of Struct Asset Copier Float Channel Structures=Get Float Data()
    NewItem:"Struct Asset Copier Float Channel Structure (by ref)"=@n11
   ) -> (then, ReturnValue)
  ]}
 GetSkeletalAnimData {nodes [
   n1 = GetSkeletalAnimData!() -> (then, Track:Movie Scene Track Object Reference) {is=title K2Node_FunctionEntry}
   > n2 = GetSections(Track=@n1)
   > n3 = For Each Loop(Array:Array of Movie Scene Section Object References=@n2) -> (Array Element:Movie Scene Section Object Reference)
   n4 = GetDisplayName(execute=@n3.Completed, Track=Get Track())
   > Return Node!(
    execute
    AnimTracks_TrackName_7_#54:String="To String (Text)"(InText=@n4)
    AnimTracks_AnimSection_3_#55:Array of Struct Asset Copier Anim Section Structures=Get AnimSections() -> (AnimSections:Array of Struct Asset Copier Anim Section Structures)
   )
   n5 = Cast To MovieSceneSkeletalAnimationSection(execute=@n3.LoopBody, Object=@n3)
   > Add(
    TargetArray:Array of Movie Scene Skeletal Animation Params Structures=Get AnimParameters() -> (AnimParameters:Array of Movie Scene Skeletal Animation Params Structures)
    NewItem:"Movie Scene Skeletal Animation Params Structure (by ref)"=Get Params!(self:Movie Scene Skeletal Animation Section Object Reference=@n5) -> (Params:Movie Scene Skeletal Animation Params Structure)
   )
   > n6 = GetStartFrame!(execute, Section:Movie Scene Section Object Reference=@n5) -> (then, ReturnValue:int=0)
   > Add(TargetArray:Array of Integers=Get SecStrtFrames() -> (SecStrtFrames:Array of Integers), NewItem:"Integer (by ref)"=@n6)
   > n7 = GetEndFrame!(execute, Section:Movie Scene Section Object Reference=@n5) -> (then, ReturnValue:int=0)
   > Add(TargetArray=Get SecNdFrames() -> (SecNdFrames:Array of Integers), NewItem=@n7)
   > Add(
    TargetArray=Get RowIndex'() -> (RowIndex':Array of Integers)
    NewItem=GetRowIndex!(self:Movie Scene Section Object Reference=@n5) -> (ReturnValue:int=0)
   )
   > Add!(
    execute
    TargetArray:Array of Struct Asset Copier Anim Section Structures=Get AnimSections()
    NewItem_SectionStartFrames_11_#23:Array of Integers=Get SecStrtFrames()
    NewItem_SectionEndFrames_12_#24:Array of Integers=Get SecNdFrames()
    NewItem_AnimParams_27_#25:Array of Movie Scene Skeletal Animation Params Structures=Get AnimParameters()
    NewItem_RowIndex_33_#26:Array of Integers=Get RowIndex'()
   ) -> (then, ReturnValue)
  ]}
 GetSpawnTrackData {nodes [
   n1 = GetSpawnTrackData!() -> (then, Track:Movie Scene Track Object Reference) {is=title K2Node_FunctionEntry}
   n2 = Cast To MovieSceneBoolTrack(Object=@n1)
   > n3 = GetPropertyName(Track=@n2) -> (ReturnValue=None)
   > GetPropertyPath(Track=@n2)
   > Return Node!(
    execute
    BoolSection_BoolSection_13_#16:Array of Struct Asset Copier Bool Section Structures=Break Struct Asset Copier Bool Track(Struct_AssetCopier_BoolTrack=GetBoolTrackData(Track=@n1) {is=title K2Node_CallFunction}).BoolSections_18_#20
   )
   Cast To MovieSceneSpawnTrack(Object=@n1)
   Reroute Node(InputPin:Name=@n3) -> (OutputPin:Name)
   Reroute Node(InputPin:Wildcard) -> (OutputPin:Wildcard)
  ]}
 GetTransformChannelData {nodes [
   n1 = GetTransformChannelData!() -> (then, DoubleChannel:Movie Scene Scripting Double Channel Object Reference) {is=title K2Node_FunctionEntry}
   n2 = "Get Keys (Double)"!(self:Movie Scene Scripting Double Channel Object Reference=@n1) -> (ReturnValue:Array of Movie Scene Scripting Key Object References) {is=title K2Node_CallFunction}
   > n3 = Branch(Condition="integer > integer"(A=Length(TargetArray=@n2)))
   > n4 = For Each Loop(Array:Array of Movie Scene Scripting Key Object References=@n2) -> (Array Element:Movie Scene Scripting Key Object Reference)
   n5 = GetDisplayName!(Object:Object Reference=@n1) -> (ReturnValue:String)
   Return Node!(
    execute=[@n7.then, @n4.Completed]
    TransformData_OwnerName_19_#39:String=Left(SourceString=@n5, Count=FindSubstring(SearchIn=@n5, Substring=_))
    TransformData_FrameTime_27_#40:Array of Frame Time Structures=Get FrameTimes()
    TransformData_FrameValue_33_#41:"Array of Float (double-precision)s"=Get FrameValues()
    TransformData_FloatKeyParams_32_#42:Array of Struct Asset Copier Float Key Params Structures=Get FloatKeyParams()
   )
   > n6 = Cast To MovieSceneScriptingDoubleKey!(execute, Object=@n4) -> (then, CastFailed, AsMovie Scene Scripting Double Key:Movie Scene Scripting Double Key Object Reference) {is=title K2Node_DynamicCast}
   > Add!(
    execute
    TargetArray:Array of Frame Time Structures=Get FrameTimes()
    NewItem:"Frame Time Structure (by ref)"="Get Time (Double)"!(self:Movie Scene Scripting Double Key Object Reference=@n6, TimeUnit:EMovieSceneTimeUnit Enum=DisplayRate) -> (ReturnValue:Frame Time Structure) {is=title K2Node_CallFunction}
   ) -> (then, ReturnValue)
   > Add(
    TargetArray:"Array of Float (double-precision)s"=Get FrameValues()
    NewItem:"Float (double-precision) (by ref)"="Get Value (Double)"!(self:Movie Scene Scripting Double Key Object Reference=@n6) -> (ReturnValue:double) {is=title K2Node_CallFunction}
   )
   > Add(
    TargetArray:Array of Struct Asset Copier Float Key Params Structures=Get FloatKeyParams()
    NewItem:"Struct Asset Copier Float Key Params Structure (by ref)"=Make Struct Asset Copier Float Key Params(
     InterpolationMode_2_#43=GetInterpolationMode(self:Movie Scene Scripting Double Key Object Reference=@n6) -> (ReturnValue=RCIM_Linear)
     TangentMode_18_#44=GetTangentMode(self:Movie Scene Scripting Double Key Object Reference=@n6) -> (ReturnValue=RCTM_Auto)
     TangentWeightMode_14_#45=GetTangentWeightMode(self:Movie Scene Scripting Double Key Object Reference=@n6) -> (ReturnValue=RCTWM_WeightedNone)
     ArriveTangent_7_#46=GetArriveTangent(self:Movie Scene Scripting Double Key Object Reference=@n6)
     ArriveTangentWeight_9_#47=GetArriveTangentWeight(self:Movie Scene Scripting Double Key Object Reference=@n6)
     LeaveTangent_13_#48=GetLeaveTangent(self:Movie Scene Scripting Double Key Object Reference=@n6)
     LeaveTangentWeight_17_#49=GetLeaveTangentWeight(self:Movie Scene Scripting Double Key Object Reference=@n6)
    )
   )
   n7 = Add(
    execute=@n3.else
    TargetArray:"Array of Float (double-precision)s"=Get FrameValues()
    NewItem:"Float (double-precision) (by ref)"="Get Default (Double)"!(self:Movie Scene Scripting Double Channel Object Reference=@n1) -> (ReturnValue:double) {is=title K2Node_CallFunction}
   )
  ]}
 GetTransformTrackData {nodes [
   n1 = GetTransformTrackData!() -> (then, Track:Movie Scene Track Object Reference) {is=title K2Node_FunctionEntry}
   > n2 = GetSections(Track=@n1)
   > n3 = For Each Loop(Array:Array of Movie Scene Section Object References=@n2) -> (Array Element:Movie Scene Section Object Reference)
   n4 = Cast To MovieScene3DTransformTrack(execute=@n3.Completed, Object=Get Track())
   > n5 = GetPropertyName(Track=@n4) -> (ReturnValue=None)
   > n6 = GetPropertyPath(Track=@n4)
   > Return Node!(
    execute
    TransformSections:Array of Struct Asset Copier Transform Section Structures=Get TransformSectionsData() -> (TransformSectionsData:Array of Struct Asset Copier Transform Section Structures)
    PropertyName:Name=@n5
    PropertyPath:String=@n6
   )
   n7 = GetChannelsByType(execute=@n3.LoopBody, Section=@n3, ChannelType=/Script/SequencerScripting.MovieSceneScriptingDoubleChannel)
   > n8 = For Each Loop(Array:Array of Movie Scene Scripting Channel Object References=@n7) -> (Array Element:Movie Scene Scripting Channel Object Reference)
   Add!(
    execute=@n8.Completed
    TargetArray:Array of Struct Asset Copier Transform Section Structures=Get TransformSectionsData()
    NewItem_BlendType_8_#34:EMovieSceneBlendType Enum=GetBlendType(self=@n3) -> (ReturnValue_bIsValid=~)
    NewItem_FloatData_12_#35:Array of Struct Asset Copier Float Channel Structures=Get DataFloat() -> (DataFloat:Array of Struct Asset Copier Float Channel Structures)
   ) -> (then, ReturnValue)
   > Clear(TargetArray=Get DataFloat())
   n9 = Cast To MovieSceneScriptingDoubleChannel!(execute=@n8.LoopBody, Object=@n8) -> (then, CastFailed, AsMovie Scene Scripting Double Channel:Movie Scene Scripting Double Channel Object Reference) {is=title K2Node_DynamicCast}
   > n10 = GetTransformChannelData!(execute, self, DoubleChannel:Movie Scene Scripting Double Channel Object Reference=@n9) -> (then, TransformData:Struct Asset Copier Float Channel Structure) {is=title K2Node_CallFunction}
   > Add!(
    execute
    TargetArray:Array of Struct Asset Copier Float Channel Structures=Get DataFloat()
    NewItem:"Struct Asset Copier Float Channel Structure (by ref)"=@n10
   ) -> (then, ReturnValue)
  ]}
 OverwriteBinding {nodes [
   OverwriteBinding!() -> (then, CurrentSourceBinding:Struct Asset Copier Binding Data Structure) {is=title K2Node_FunctionEntry}
   n1 = Get Target Level Sequence()
   > Is Valid(InputObject=@n1)
   > n2 = GetBindings(Sequence=@n1)
   > n3 = For Each Loop(Array:Array of Movie Scene Binding Proxy Structures=@n2) -> (Array Element:Movie Scene Binding Proxy Structure)
   > n4 = Set CurrentTargetBinding(CurrentTargetBinding:Movie Scene Binding Proxy Structure=@n3) -> (Output_Get:Movie Scene Binding Proxy Structure)
   > n5 = GetDisplayName!(execute:Exec, InBinding:"Movie Scene Binding Proxy Structure (by ref)"=@n4) -> (then:Exec, ReturnValue:Text)
   n6 = Get() -> (CurrentSourceBinding_Overwrite_40_#2=~)
   > n7 = Branch(Condition=IsEmpty(InString=@n6.CurrentSourceBinding_ParentAssetName_21_#4))
   Branch(execute=[@n7.then, @n16.then], Condition="Equal Exactly (String)"(A="To String (Text)"(InText=@n5), B=@n6))
   > n8 = Sequence!(execute) -> (then_0, then_1)
   > n9 = GetTracks(InBinding=Get CurrentTargetBinding() -> (CurrentTargetBinding:Movie Scene Binding Proxy Structure))
   > n10 = For Each Loop(Array:Array of Movie Scene Track Object References=@n9) -> (Array Element:Movie Scene Track Object Reference)
   > n11 = GetDisplayName!(execute, Track:Movie Scene Track Object Reference=@n10) -> (then, ReturnValue)
   > Add!(
    execute
    TargetMap:Map of Strings to Movie Scene Track Object References=Get TrackMap() -> (TrackMap:Map of Strings to Movie Scene Track Object References)
    Key:"String (by ref)"="To String (Text)"(InText=@n11)
    Value:"Movie Scene Track Object Reference (by ref)"=@n10
   ) -> (then) {is=title K2Node_CallFunction}
   n12 = Sequence!(execute=@n8.then_1) -> (then_0, then_1, then_2:Exec, then_3:Exec, then_4:Exec, then_5:Exec, then_6:Exec)
   n13 = Get() -> (CurrentSourceBinding_Overwrite_40_#2=~)
   > K2Node_Composite_0 = Floats!(
    Exec:Exec
    Array:Array of Struct Asset Copier Float Track Structures=@n13.CurrentSourceBinding_FloatTracks_22_#5
   ) {is=title K2Node_Composite}
   K2Node_Composite_1 = Transform!(
    Exec:Exec=@n12.then_1
    Struct_AssetCopier_TransformTrack:"Struct Asset Copier Transform Track Structure (by ref)"=@n13.CurrentSourceBinding_TransformTrack_43_#6
   ) {is=title K2Node_Composite}
   K2Node_Composite_2 = Bools!(
    Exec:Exec=@n12.then_2
    Array:Array of Struct Asset Copier Bool Track Structures=@n13.CurrentSourceBinding_BoolTracks_44_#7
   ) {is=title K2Node_Composite}
   K2Node_Composite_3 = Anim!(
    Exec:Exec=@n12.then_3
    Array:Array of Struct Asset Copier Anim Track Structures=@n13.CurrentSourceBinding_AnimTracks_46_#8
   ) {is=title K2Node_Composite}
   K2Node_Composite_4 = Spawns!(
    Exec:Exec=@n12.then_4
    Array:Array of Struct Asset Copier Spawn Track Structures=@n13.CurrentSourceBinding_SpawnTracks_45_#9
   ) {is=title K2Node_Composite}
   K2Node_Composite_5 = CtrlRigs!(
    Exec:Exec=@n12.then_5
    Array:Array of Struct Asset Copier Control Rig Track Structures=@n13.CurrentSourceBinding_ControlRigTracks_35_#10
    Current Source Binding Control Rig Binding:Control Rig Sequencer Binding Proxy Structure
   ) {is=title K2Node_Composite}
   Return Node!(execute=@n12.then_6, OutTargetBinding:Movie Scene Binding Proxy Structure=Get CurrentTargetBinding())
   n14 = GetParent(execute=@n7.else, InBinding=Get CurrentTargetBinding())
   > n15 = GetDisplayName!(execute, InBinding:"Movie Scene Binding Proxy Structure (by ref)"=@n14) -> (then, ReturnValue)
   > n16 = Branch(Condition="Equal Exactly (String)"(A="To String (Text)"(InText=@n15), B=@n6.CurrentSourceBinding_ParentAssetName_21_#4))
  ]}
 Floats {
  path=OverwriteBinding.K2Node_Composite_0.Floats
  nodes [
   n1 = Inputs!() -> (Exec:Exec, Array:Array of Struct Asset Copier Float Track Structures)
   > Clear!(
    execute
    TargetMap:Map of Integers to Movie Scene Section Object References=Get SectionMap() -> (SectionMap:Map of Integers to Movie Scene Section Object References)
   ) -> (then) {is=title K2Node_CallFunction}
   > K2Node_Composite_0 = RemoveFloatTracksNotInSource!(execute, Array:Array of Struct Asset Copier Float Track Structures=@n1) -> (Completed:Exec) {is=title K2Node_Composite}
   > n2 = For Each Loop!(Exec, Array:Array of Struct Asset Copier Float Track Structures=@n1) -> (
    LoopBody
    Array Element_DisplayName_18_#50:String
    Array Element_PropertyName_22_#51:Name
    Array Element_PropertyPath_24_#52:String
    Array Element_FloatSections_13_#53:Array of Struct Asset Copier Float Section Structures
    Array Index
    Completed
   )
   n3 = Find(TargetMap:Map of Strings to Movie Scene Track Object References=Get TrackMap(), Key=@n2) -> (Value:Movie Scene Track Object Reference)
   > n4 = Branch(Condition=@n3.ReturnValue)
   > n5 = Cast To MovieSceneFloatTrack(Object=@n3)
   > n6 = Sequence!(execute) -> (then_0, then_1)
   > n7 = GetSections(Track=@n5)
   > n8 = For Each Loop!(Exec, Array:Array of Movie Scene Section Object References=@n7) -> (LoopBody, Array Element:Movie Scene Section Object Reference, Array Index, Completed)
   > Add(
    TargetMap:Map of Integers to Movie Scene Section Object References=Get SectionMap()
    Key:"Integer (by ref)"=@n8.Array Index
    Value:"Movie Scene Section Object Reference (by ref)"=@n8
   )
   n9 = For Each Loop!(
    Exec=@n6.then_1
    Array:Array of Struct Asset Copier Float Section Structures=@n2.Array Element_FloatSections_13_#53
   ) -> (
    LoopBody
    Array Element_BlendType_5_#32:EMovieSceneBlendType Enum
    Array Element_FloatData_10_#33:Array of Struct Asset Copier Float Channel Structures
    Array Index
    Completed
   )
   n10 = Find(TargetMap:Map of Integers to Movie Scene Section Object References=Get SectionMap(), Key:"Integer (by ref)"=@n9.Array Index) -> (Value:Movie Scene Section Object Reference)
   > n11 = Branch(Condition=@n10.ReturnValue)
   > SetBlendType(self=@n10, InBlendType=@n9)
   > SetFloatKeys(Section=@n10, FloatChannels=@n9.Array Element_FloatData_10_#33)
   AddFloatSection(
    execute=@n11.else
    Track=@n5
    InBlendType=@n9
    FloatChannels=@n9.Array Element_FloatData_10_#33
    SectionRowIndex=@n9.Array Index
   )
   CreateFloatTrack!(
    execute=@n4.else
    self
    TrackSections:Array of Struct Asset Copier Float Section Structures=@n2.Array Element_FloatSections_13_#53
    TrackName:String=@n2
    InBinding:"Movie Scene Binding Proxy Structure (by ref)"=Get CurrentTargetBinding()
    PropertyName:"Name (by ref)"=@n2.Array Element_PropertyName_22_#51
    PropertyPath:String=@n2.Array Element_PropertyPath_24_#52
   ) -> (then) {is=title K2Node_CallFunction}
   Outputs()
  ]
 }
 RemoveFloatTracksNotInSource {
  path=OverwriteBinding.K2Node_Composite_0.Floats.K2Node_Composite_0.RemoveFloatTracksNotInSource
  nodes [
   n1 = Inputs!() -> (execute:Exec, Array)
   > n2 = GetTracks(InBinding=Get CurrentTargetBinding())
   > n3 = For Each Loop!(Exec, Array:Array of Movie Scene Track Object References=@n2) -> (LoopBody, Array Element:Movie Scene Track Object Reference, Array Index, Completed)
   Outputs!(Completed:Exec=@n3.Completed)
   > n4 = Cast To MovieSceneFloatTrack(Object=@n3)
   > n5 = GetDisplayName!(execute, Track:Movie Scene Track Object Reference=@n4) -> (then, ReturnValue)
   > n6 = For Each Loop with Break!(Exec:Exec, Array:Array of Struct Asset Copier Float Track Structures=@n1, Break:Exec=@n8.then) -> (
    LoopBody:Exec
    Array Element_DisplayName_18_#50:String
    Array Element_PropertyName_22_#51:Name
    Array Element_PropertyPath_24_#52:String
    Array Element_FloatSections_13_#53:Array of Struct Asset Copier Float Section Structures
    Array Index:int
    Completed:Exec
   )
   n7 = Branch(execute=@n6.Completed, Condition=Get TrackFound() -> (TrackFound:bool))
   > Set TrackFound(TrackFound:bool=false) -> (Output_Get:bool)
   RemoveTrack!(
    execute=@n7.else
    InBinding:"Movie Scene Binding Proxy Structure (by ref)"=Get CurrentTargetBinding()
    TrackToRemove:Movie Scene Track Object Reference=@n4
   ) -> (then)
   Branch(execute=@n6.LoopBody, Condition="Equal Exactly (String)"(A="To String (Text)"(InText=@n5), B=@n6))
   > n8 = Set TrackFound(TrackFound=true)
  ]
 }
 Transform {
  path=OverwriteBinding.K2Node_Composite_1.Transform
  nodes [
   n1 = Inputs!() -> (Exec:Exec, Struct_AssetCopier_TransformTrack:"Struct Asset Copier Transform Track Structure (by ref)")
   > Clear(TargetMap=Get SectionMap())
   n2 = Break Struct Asset Copier Transform Track!(Struct_AssetCopier_TransformTrack:"Struct Asset Copier Transform Track Structure (by ref)"=@n1) -> (
    TransformSections_8_#13:Array of Struct Asset Copier Transform Section Structures
    PropertyName_11_#14:Name
    PropertyPath_14_#15:String
   ) {is=title K2Node_BreakStruct}
   > n3 = For Each Loop!(Exec, Array:Array of Struct Asset Copier Transform Section Structures=@n2) -> (
    LoopBody
    Array Element_BlendType_8_#34:EMovieSceneBlendType Enum
    Array Element_FloatData_12_#35:Array of Struct Asset Copier Float Channel Structures
    Array Index
    Completed
   )
   n4 = Find(TargetMap:Map of Strings to Movie Scene Track Object References=Get TrackMap(), Key:"String (by ref)"=Transform) -> (Value:Movie Scene Track Object Reference)
   > n5 = Branch(Condition=@n4.ReturnValue)
   > n6 = Cast To MovieScene3DTransformTrack(Object=@n4)
   > n7 = Sequence()
   > n8 = GetSections(Track=@n6)
   > n9 = For Each Loop!(Exec, Array:Array of Movie Scene Section Object References=@n8) -> (LoopBody, Array Element:Movie Scene Section Object Reference, Array Index, Completed)
   > Add(TargetMap=Get SectionMap(), Key=@n9.Array Index, Value=@n9)
   n10 = Find(TargetMap:Map of Integers to Movie Scene Section Object References=Get SectionMap(), Key:"Integer (by ref)"=@n3.Array Index) -> (Value:Movie Scene Section Object Reference)
   n11 = Branch(execute=@n7.then_1, Condition=@n10.ReturnValue)
   > SetBlendType(self=@n10, InBlendType=@n3)
   > SetTransformFloatKeys(Section=@n10, FloatData=@n3.Array Element_FloatData_12_#35)
   AddTransformSection(
    execute=@n11.else
    InBlendType=@n3
    FloatData=@n3.Array Element_FloatData_12_#35
    SectionRowIndex=@n3.Array Index
    Track=@n6
   )
   CreateTransformTrack!(
    execute=@n5.else
    self
    InBinding:"Movie Scene Binding Proxy Structure (by ref)"=Get CurrentTargetBinding()
    TransformData:Array of Struct Asset Copier Transform Section Structures=@n2
    PropertyName:Name=@n2.PropertyName_11_#14
    PropertyPath:String=@n2.PropertyPath_14_#15
   ) -> (then) {is=title K2Node_CallFunction}
   Outputs!()
  ]
 }
 Bools {
  path=OverwriteBinding.K2Node_Composite_2.Bools
  nodes [
   n1 = Inputs!() -> (Exec, Array:Array of Struct Asset Copier Bool Track Structures)
   > Clear(TargetMap=Get SectionMap())
   > K2Node_Composite_0 = RemoveBoolTracksNotInSource!(execute, InputPin:Array of Struct Asset Copier Bool Track Structures=@n1) -> (Completed:Exec) {is=title K2Node_Composite}
   > n2 = For Each Loop!(Exec, Array:Array of Struct Asset Copier Bool Track Structures=@n1) -> (
    LoopBody
    Array Element_TrackName_8_#17:String
    Array Element_PropertyName_16_#18:Name
    Array Element_PropertyPath_17_#19:String
    Array Element_BoolSections_18_#20:Array of Struct Asset Copier Bool Section Structures
    Array Index
    Completed
   )
   n3 = Find(TargetMap:Map of Strings to Movie Scene Track Object References=Get TrackMap(), Key:"String (by ref)"=@n2) -> (Value:Movie Scene Track Object Reference)
   > n4 = Branch(Condition=@n3.ReturnValue)
   > n5 = Cast To MovieSceneBoolTrack(Object=@n3)
   > n6 = Sequence()
   > n7 = GetSections(Track=@n5)
   > n8 = For Each Loop!(Exec, Array:Array of Movie Scene Section Object References=@n7) -> (LoopBody, Array Element:Movie Scene Section Object Reference, Array Index, Completed)
   > Add(TargetMap=Get SectionMap(), Key=@n8.Array Index, Value=@n8)
   n9 = For Each Loop!(
    Exec=@n6.then_1
    Array:Array of Struct Asset Copier Bool Section Structures=@n2.Array Element_BoolSections_18_#20
   ) -> (
    LoopBody
    Array Element_BoolChannel_5_#27:Array of Struct Asset Copier Bool Channel Structures
    Array Index
    Completed
   )
   n10 = Find(TargetMap:Map of Integers to Movie Scene Section Object References=Get SectionMap(), Key:"Integer (by ref)"=@n9.Array Index) -> (Value:Movie Scene Section Object Reference)
   > n11 = Branch(Condition=@n10.ReturnValue)
   > SetBoolKeys(Section=@n10, BoolChannels=@n9)
   AddBoolSection(execute=@n11.else, Track=@n5, BoolChannels=@n9)
   CreateBoolTrack!(
    execute=@n4.else
    self
    TrackSections:Array of Struct Asset Copier Bool Section Structures=@n2.Array Element_BoolSections_18_#20
    TrackName:String=@n2
    InBinding:"Movie Scene Binding Proxy Structure (by ref)"=Get CurrentTargetBinding()
    PropertyName:Name=@n2.Array Element_PropertyName_16_#18
    PropertyPath:String=@n2.Array Element_PropertyPath_17_#19
   ) -> (then) {is=title K2Node_CallFunction}
   Outputs()
  ]
 }
 RemoveBoolTracksNotInSource {
  path=OverwriteBinding.K2Node_Composite_2.Bools.K2Node_Composite_0.RemoveBoolTracksNotInSource
  nodes [
   n1 = Inputs!() -> (execute:Exec, InputPin:Array of Struct Asset Copier Bool Track Structures)
   > n2 = GetTracks(InBinding=Get CurrentTargetBinding())
   > n3 = For Each Loop!(Exec, Array:Array of Movie Scene Track Object References=@n2) -> (LoopBody, Array Element:Movie Scene Track Object Reference, Array Index, Completed)
   Outputs!(Completed:Exec=@n3.Completed)
   > n4 = Cast To MovieSceneBoolTrack(Object=@n3)
   > n5 = GetDisplayName(Track=@n4)
   > n6 = For Each Loop with Break!(Exec, Array:Array of Struct Asset Copier Bool Track Structures=@n1, Break=@n8.then) -> (
    LoopBody
    Array Element_TrackName_8_#17:String
    Array Element_PropertyName_16_#18:Name
    Array Element_PropertyPath_17_#19:String
    Array Element_BoolSections_18_#20:Array of Struct Asset Copier Bool Section Structures
    Array Index
    Completed
   )
   n7 = Branch(execute=@n6.Completed, Condition=Get TrackFound())
   > Set TrackFound(TrackFound=false)
   RemoveTrack(execute=@n7.else, InBinding=Get CurrentTargetBinding(), TrackToRemove=@n4)
   Branch(execute=@n6.LoopBody, Condition="Equal Exactly (String)"(A="To String (Text)"(InText=@n5), B=@n6))
   > n8 = Set TrackFound(TrackFound=true)
  ]
 }
 Anim {
  path=OverwriteBinding.K2Node_Composite_3.Anim
  nodes [
   n1 = Inputs!() -> (Exec:Exec, Array:Array of Struct Asset Copier Anim Track Structures)
   > Clear(TargetMap=Get SectionMap())
   > n2 = GetTracks(InBinding=Get CurrentTargetBinding())
   > n3 = For Each Loop(Array=@n2)
   > n4 = Cast To MovieSceneSkeletalAnimationTrack(Object=@n3)
   > RemoveTrack(InBinding=Get CurrentTargetBinding(), TrackToRemove=@n4)
   > n5 = For Each Loop!(Exec, Array:Array of Struct Asset Copier Anim Track Structures=@n1) -> (
    LoopBody
    Array Element_TrackName_7_#54:String
    Array Element_AnimSection_3_#55:Array of Struct Asset Copier Anim Section Structures
    Array Index
    Completed
   )
   > CreateAnimTrack!(
    execute
    self
    InBinding:"Movie Scene Binding Proxy Structure (by ref)"=Get CurrentTargetBinding()
    TrackName:String=@n5
    TrackSections:Array of Struct Asset Copier Anim Section Structures=@n5.Array Element_AnimSection_3_#55
   ) -> (then) {is=title K2Node_CallFunction}
   Outputs!()
  ]
 }
 Spawns {
  path=OverwriteBinding.K2Node_Composite_4.Spawns
  nodes [
   n1 = Inputs() -> (Array:Array of Struct Asset Copier Spawn Track Structures)
   > Clear(TargetMap=Get SectionMap())
   > n2 = For Each Loop!(Exec, Array:Array of Struct Asset Copier Spawn Track Structures=@n1) -> (
    LoopBody
    Array Element_BoolSection_13_#16:Array of Struct Asset Copier Bool Section Structures
    Array Index
    Completed
   )
   > n3 = For Each Loop!(Exec, Array:Array of Struct Asset Copier Bool Section Structures=@n2) -> (
    LoopBody
    Array Element_BoolChannel_5_#27:Array of Struct Asset Copier Bool Channel Structures
    Array Index
    Completed
   )
   n4 = Find(TargetMap:Map of Strings to Movie Scene Track Object References=Get TrackMap(), Key:"String (by ref)"=Spawned) -> (Value:Movie Scene Track Object Reference)
   > n5 = Branch(Condition=@n4.ReturnValue)
   > n6 = Cast To MovieSceneSpawnTrack(Object=@n4)
   > n7 = Sequence()
   > n8 = GetSections(Track=@n6)
   > n9 = For Each Loop!(Exec, Array:Array of Movie Scene Section Object References=@n8) -> (LoopBody, Array Element:Movie Scene Section Object Reference, Array Index, Completed)
   > Add(TargetMap=Get SectionMap(), Key=@n9.Array Index, Value=@n9)
   n10 = Find(TargetMap:Map of Integers to Movie Scene Section Object References=Get SectionMap(), Key:"Integer (by ref)"=@n3.Array Index) -> (Value:Movie Scene Section Object Reference)
   n11 = Branch(execute=@n7.then_1, Condition=@n10.ReturnValue)
   > n12 = Cast To MovieSceneBoolSection(Object=@n10)
   > SetBoolKeys(Section=@n12, BoolChannels=@n3)
   addSpawnedSection(execute=@n11.else, Track=@n6)
   CreateSpawnTrack!(
    execute=@n5.else
    self
    SectionChannels:Array of Struct Asset Copier Bool Channel Structures=@n3
    InBinding:"Movie Scene Binding Proxy Structure (by ref)"=Get CurrentTargetBinding()
   ) -> (then) {is=title K2Node_CallFunction}
   Outputs()
  ]
 }
 CtrlRigs {
  path=OverwriteBinding.K2Node_Composite_5.CtrlRigs
  nodes [
   n1 = Inputs!() -> (
    Exec
    Array:Array of Struct Asset Copier Control Rig Track Structures
    Current Source Binding Control Rig Binding:Control Rig Sequencer Binding Proxy Structure
   )
   > Clear(TargetMap=Get SectionMap())
   > K2Node_Composite_0 = RemoveCtrlRigTracksNotInSource!(execute, InputPin:Array of Struct Asset Copier Control Rig Track Structures=@n1) -> (Completed:Exec) {is=title K2Node_Composite}
   > n2 = For Each Loop!(Exec, Array:Array of Struct Asset Copier Control Rig Track Structures=@n1) -> (
    LoopBody
    Array Element_ControlRigName_2_#21:String
    Array Element_ControlRigSections_8_#22:Array of Struct Asset Copier Control Rig Section Structures
    Array Index
    Completed
   )
   n3 = Find(TargetMap:Map of Strings to Movie Scene Track Object References=Get TrackMap(), Key:"String (by ref)"=@n2) -> (Value:Movie Scene Track Object Reference)
   > n4 = Branch(Condition=@n3.ReturnValue)
   > n5 = Cast To MovieSceneControlRigParameterTrack(Object=@n3)
   > n6 = Sequence()
   > n7 = GetSections(Track=@n5)
   > n8 = For Each Loop!(Exec, Array:Array of Movie Scene Section Object References=@n7) -> (LoopBody, Array Element:Movie Scene Section Object Reference, Array Index, Completed)
   > Add(TargetMap=Get SectionMap(), Key=@n8.Array Index, Value=@n8)
   n9 = For Each Loop!(
    Exec=@n6.then_1
    Array:Array of Struct Asset Copier Control Rig Section Structures=@n2.Array Element_ControlRigSections_8_#22
   ) -> (
    LoopBody
    Array Element_BlendType_20_#28:EMovieSceneBlendType Enum
    Array Element_RowDisplayName_30_#29:Text
    Array Element_FloatControls_13_#30:Array of Struct Asset Copier Float Channel Structures
    Array Element_BoolControls_14_#31:Array of Struct Asset Copier Bool Channel Structures
    Array Index
    Completed
   )
   > SetTrackRowDisplayName(Track=@n5, InName=@n9.Array Element_RowDisplayName_30_#29, RowIndex=@n9.Array Index)
   n10 = Find(TargetMap:Map of Integers to Movie Scene Section Object References=Get SectionMap(), Key:"Integer (by ref)"=@n9.Array Index) -> (Value:Movie Scene Section Object Reference)
   > n11 = Branch(Condition=@n10.ReturnValue)
   > n12 = Cast To MovieSceneControlRigParameterSection(Object=@n10)
   > SetBlendType(self=@n12, InBlendType=@n9)
   > SetFloatKeys(Section=@n12, FloatChannels=@n9.Array Element_FloatControls_13_#30)
   > SetBoolKeys(Section=@n12, BoolChannels=@n9.Array Element_BoolControls_14_#31)
   AddCtrlRigSection(
    execute=@n11.else
    Track=@n5
    InBlendType=@n9
    TrackRowName=@n9.Array Element_RowDisplayName_30_#29
    FloatChannels=@n9.Array Element_FloatControls_13_#30
    BoolChannels=@n9.Array Element_BoolControls_14_#31
    SectionRowIndex=@n9.Array Index
   )
   CreateCtrlRigTrack!(
    execute=@n4.else
    self
    InBinding:"Movie Scene Binding Proxy Structure (by ref)"=Get CurrentTargetBinding()
    CtrlRigSections:Array of Struct Asset Copier Control Rig Section Structures=@n2.Array Element_ControlRigSections_8_#22
    TrackName:String=@n2
   ) -> (then) {is=title K2Node_CallFunction}
   Outputs()
  ]
 }
 RemoveCtrlRigTracksNotInSource {
  path=OverwriteBinding.K2Node_Composite_5.CtrlRigs.K2Node_Composite_0.RemoveCtrlRigTracksNotInSource
  nodes [
   n1 = Inputs!() -> (execute:Exec, InputPin:Array of Struct Asset Copier Control Rig Track Structures)
   > n2 = GetTracks(InBinding=Get CurrentTargetBinding())
   > n3 = For Each Loop!(Exec, Array:Array of Movie Scene Track Object References=@n2) -> (LoopBody, Array Element:Movie Scene Track Object Reference, Array Index, Completed)
   Outputs!(Completed:Exec=@n3.Completed)
   > n4 = Cast To MovieSceneControlRigParameterTrack(Object=@n3)
   > n5 = GetDisplayName(Track=@n4)
   > n6 = For Each Loop with Break!(Exec, Array:Array of Struct Asset Copier Control Rig Track Structures=@n1, Break=@n8.then) -> (
    LoopBody
    Array Element_ControlRigName_2_#21:String
    Array Element_ControlRigSections_8_#22:Array of Struct Asset Copier Control Rig Section Structures
    Array Index
    Completed
   )
   n7 = Branch(execute=@n6.Completed, Condition=Get TrackFound())
   > Set TrackFound(TrackFound=false)
   RemoveTrack(execute=@n7.else, InBinding=Get CurrentTargetBinding(), TrackToRemove=@n4)
   Branch(execute=@n6.LoopBody, Condition="Equal Exactly (String)"(A="To String (Text)"(InText=@n5), B=@n6))
   > n8 = Set TrackFound(TrackFound=true)
  ]
 }
 PasteAssets {nodes [
   PasteAssets!() -> (then, CurrentBindingData:Struct Asset Copier Binding Data Structure) {is=title K2Node_FunctionEntry}
   > n1 = Sequence()
   > n2 = Set CurrentBinding!(
    execute
    CurrentBinding:Struct Asset Copier Binding Data Structure=Get CurrentBindingData() -> (CurrentBindingData:Struct Asset Copier Binding Data Structure)
   ) -> (
    then
    Output_Get_BindingOwnerName_31_#1:String
    Output_Get_Overwrite_40_#2:bool=~
    Output_Get_ObjectTemplate_53_#3:Object Reference
    Output_Get_ParentAssetName_21_#4:String
    Output_Get_FloatTracks_22_#5:Array of Struct Asset Copier Float Track Structures
    Output_Get_TransformTrack_43_#6:Struct Asset Copier Transform Track Structure
    Output_Get_BoolTracks_44_#7:Array of Struct Asset Copier Bool Track Structures
    Output_Get_AnimTracks_46_#8:Array of Struct Asset Copier Anim Track Structures
    Output_Get_SpawnTracks_45_#9:Array of Struct Asset Copier Spawn Track Structures
    Output_Get_ControlRigTracks_35_#10:Array of Struct Asset Copier Control Rig Track Structures
    Output_Get_ChildBindings_51_#11:Array of Movie Scene Binding Proxy Structures
    Output_Get_ControlRigBinding_58_#12:Control Rig Sequencer Binding Proxy Structure
   )
   > n3 = Branch(Condition=@n2.Output_Get_Overwrite_40_#2)
   n4 = Get CurrentBindingData()
   > n5 = OverwriteBinding(CurrentSourceBinding=@n4) {is=title K2Node_CallFunction}
   > Set ParentBinding(ParentBinding:Movie Scene Binding Proxy Structure=@n5) -> (Output_Get:Movie Scene Binding Proxy Structure)
   n6 = GenerateNewBinding!(execute=@n3.else, self, CurrentSourceBinding:Struct Asset Copier Binding Data Structure=@n4) -> (then, OutTargetBinding:Movie Scene Binding Proxy Structure) {is=title K2Node_CallFunction}
   > Set ParentBinding(ParentBinding=@n6)
   n7 = For Each Loop(Exec=@n1.then_1, Array:Array of Movie Scene Binding Proxy Structures=@n2.Output_Get_ChildBindings_51_#11) -> (Array Element:Movie Scene Binding Proxy Structure)
   > n8 = CollectBindingData(InBinding=@n7, Overwrite=true, ParentBinding=Get ParentBinding() -> (ParentBinding:Movie Scene Binding Proxy Structure))
   > PasteAssets(CurrentBindingData=@n8) {is=title K2Node_CallFunction}
  ]}
 PreLoadSource {nodes [
   PreLoadSource() {is=title K2Node_FunctionEntry}
   n1 = Get Source Level Sequence()
   n2 = Get BeenOpened() -> (BeenOpened:Array of Level Sequence Object References)
   > Branch(Condition="integer < integer"(A=Find Item!(TargetArray:Array of Level Sequence Object References=@n2, ItemToFind:"Level Sequence Object Reference (by ref)"=@n1) -> (ReturnValue)))
   > OpenLevelSequence(LevelSequence=Get Source Level Sequence())
   > Add(
    TargetArray:Array of Level Sequence Object References=Get BeenOpened()
    NewItem:"Level Sequence Object Reference (by ref)"=Get Source Level Sequence()
   ) {is=title K2Node_CallArrayFunction}
  ]}
 SetAnimSections {nodes [SetAnimSections!() -> (
    then
    Object:Object Reference
    AnimSection_SectionStartFrames_11_#23:Array of Integers
    AnimSection_SectionEndFrames_12_#24:Array of Integers
    AnimSection_AnimParams_27_#25:Array of Movie Scene Skeletal Animation Params Structures
    AnimSection_RowIndex_33_#26:Array of Integers
   ) {is=title K2Node_FunctionEntry}]}
 SetBoolKeys {nodes [
   SetBoolKeys!() -> (then, Section:Movie Scene Section Object Reference, BoolChannels:Array of Struct Asset Copier Bool Channel Structures) {is=title K2Node_FunctionEntry}
   > n1 = Sequence()
   > n2 = GetChannelsByType(
    Section=Get Section() -> (Section:Movie Scene Section Object Reference)
    ChannelType=&MovieSceneScriptingBoolChannel
   )
   > n3 = For Each Loop(Array:Array of Movie Scene Scripting Channel Object References=@n2) -> (Array Element:Movie Scene Scripting Channel Object Reference)
   > n4 = Cast To MovieSceneScriptingBoolChannel(Object=@n3)
   n5 = "Get Keys (Bool)"(self=@n4)
   n6 = For Each Loop(Exec=[@n4.then, @n7.then], Array:Array of Movie Scene Scripting Key Object References=@n5) -> (Array Element:Movie Scene Scripting Key Object Reference)
   n7 = Branch(execute=@n6.Completed, Condition="integer > integer"(A=Length(TargetArray=@n5)))
   n8 = GetDisplayName!(Object:Object Reference=@n4) -> (ReturnValue:String)
   Add(
    execute=@n7.else
    TargetMap:Map of Strings to Movie Scene Scripting Bool Channel Object References=Get ControlsMap() -> (ControlsMap:Map of Strings to Movie Scene Scripting Bool Channel Object References)
    Key:"String (by ref)"=Left(SourceString=@n8, Count=FindSubstring(SearchIn=@n8, Substring=_))
    Value:"Movie Scene Scripting Bool Channel Object Reference (by ref)"=@n4
   ) {is=title K2Node_CallFunction}
   "Remove Key (Bool)"!(execute=@n6.LoopBody, self:Movie Scene Scripting Bool Channel Object Reference=@n4, Key:Movie Scene Scripting Key Object Reference=@n6) -> (then)
   n9 = For Each Loop!(Exec=@n1.then_1, Array:Array of Struct Asset Copier Bool Channel Structures=Get BoolChannels()) -> (
    LoopBody
    Array Element_OwnerName_19_#36:String
    Array Element_FrameTime_14_#37:Array of Frame Time Structures
    Array Element_FrameValue_7_#38:Array of Booleans
    Array Index
    Completed
   )
   n10 = Find(TargetMap:Map of Strings to Movie Scene Scripting Bool Channel Object References=Get ControlsMap(), Key:"String (by ref)"=@n9) -> (Value:Movie Scene Scripting Bool Channel Object Reference)
   > Branch(Condition=@n10.ReturnValue)
   > n11 = Branch(Condition="integer > integer"(A=Length(TargetArray:Array of Frame Time Structures=@n9.Array Element_FrameTime_14_#37)))
   > n12 = For Each Loop!(Exec, Array:Array of Frame Time Structures=@n9.Array Element_FrameTime_14_#37) -> (LoopBody, Array Element:Frame Time Structure, Array Index, Completed)
   n13 = Break Frame Time!(FrameTime:"Frame Time Structure (by ref)"=@n12) -> (FrameNumber:Frame Number Structure, SubFrame:float=~) {is=title K2Node_BreakStruct}
   > "Add Key (Bool)"!(
    execute
    self:Movie Scene Scripting Bool Channel Object Reference=@n10
    InTime:"Frame Number Structure (by ref)"=FrameRateConversion!(FrameNumber:"Frame Number Structure (by ref)"=@n13) -> (ConvertedFrameNumber:Frame Number Structure) {is=title K2Node_CallFunction}
    NewValue:bool="Get (a ref)"(Array:Array of Booleans=@n9.Array Element_FrameValue_7_#38, Dimension 1=@n12.Array Index) -> (Output:"Boolean (by ref)")
    SubFrame:float=@n13.SubFrame
    TimeUnit:EMovieSceneTimeUnit Enum=DisplayRate
   ) -> (then, ReturnValue:Movie Scene Scripting Bool Key Object Reference)
   "Set Default (Bool)"!(
    execute=@n11.else
    self:Movie Scene Scripting Bool Channel Object Reference=@n10
    InDefaultValue:bool="Get (a ref)"(Array=@n9.Array Element_FrameValue_7_#38, Dimension 1=0)
   ) -> (then) {is=title K2Node_CallFunction}
  ]}
 SetFloatKeys {nodes [
   n1 = SetFloatKeys!() -> (then, Section:Movie Scene Section Object Reference, FloatChannels:Array of Struct Asset Copier Float Channel Structures) {is=title K2Node_FunctionEntry}
   > n2 = Sequence()
   > n3 = GetChannelsByType(Section=Get Section(), ChannelType=&MovieSceneScriptingFloatChannel)
   > n4 = For Each Loop(Array:Array of Movie Scene Scripting Channel Object References=@n3) -> (Array Element:Movie Scene Scripting Channel Object Reference)
   > n5 = Cast To MovieSceneScriptingFloatChannel(Object=@n4)
   n6 = "Get Keys (Float)"(self=@n5)
   n7 = For Each Loop(Exec=[@n5.then, @n8.then], Array:Array of Movie Scene Scripting Key Object References=@n6) -> (Array Element:Movie Scene Scripting Key Object Reference)
   n8 = Branch(execute=@n7.Completed, Condition="integer > integer"(A=Length(TargetArray:Array of Movie Scene Scripting Key Object References=@n6)))
   n9 = GetDisplayName(Object=@n5)
   Add(
    execute=@n8.else
    TargetMap:Map of Strings to Movie Scene Scripting Float Channel Object References=Get ControlsMap() -> (ControlsMap:Map of Strings to Movie Scene Scripting Float Channel Object References)
    Key=Left(SourceString=@n9, Count=FindSubstring(SearchIn=@n9, Substring=_))
    Value:"Movie Scene Scripting Float Channel Object Reference (by ref)"=@n5
   )
   "Remove Key (Float)"!(execute=@n7.LoopBody, self:Movie Scene Scripting Float Channel Object Reference=@n5, Key:Movie Scene Scripting Key Object Reference=@n7) -> (then)
   n10 = For Each Loop!(Exec=@n2.then_1, Array:Array of Struct Asset Copier Float Channel Structures=@n1.FloatChannels) -> (
    LoopBody
    Array Element_OwnerName_19_#39:String
    Array Element_FrameTime_27_#40:Array of Frame Time Structures
    Array Element_FrameValue_33_#41:"Array of Float (double-precision)s"
    Array Element_FloatKeyParams_32_#42:Array of Struct Asset Copier Float Key Params Structures
    Array Index
    Completed
   )
   n11 = Find(TargetMap:Map of Strings to Movie Scene Scripting Float Channel Object References=Get ControlsMap(), Key=@n10) -> (Value:Movie Scene Scripting Float Channel Object Reference)
   > Branch(Condition=@n11.ReturnValue)
   > n12 = Branch(Condition="integer > integer"(A=Length(TargetArray:Array of Frame Time Structures=@n10.Array Element_FrameTime_27_#40)))
   > n13 = For Each Loop!(Exec, Array:Array of Frame Time Structures=@n10.Array Element_FrameTime_27_#40) -> (LoopBody, Array Element:Frame Time Structure, Array Index, Completed)
   n14 = Break Frame Time(FrameTime=@n13) -> (SubFrame=~)
   > n15 = "Add Key (Float)"!(
    execute
    self:Movie Scene Scripting Float Channel Object Reference=@n11
    InTime:"Frame Number Structure (by ref)"=FrameRateConversion(FrameNumber=@n14)
    NewValue:float="Get (a ref)"(Array:"Array of Float (double-precision)s"=@n10.Array Element_FrameValue_33_#41, Dimension 1=@n13.Array Index) -> (Output:"Float (double-precision) (by ref)")
    SubFrame:float=@n14.SubFrame
    TimeUnit:EMovieSceneTimeUnit Enum=DisplayRate
    InInterpolation:EMovieSceneKeyInterpolation Enum=Auto
   ) -> (then, ReturnValue:Movie Scene Scripting Float Key Object Reference)
   n16 = Break Struct Asset Copier Float Key Params!(Struct_AssetCopier_FloatKeyParams:"Struct Asset Copier Float Key Params Structure (by ref)"="Get (a ref)"(
     Array:Array of Struct Asset Copier Float Key Params Structures=@n10.Array Element_FloatKeyParams_32_#42
     Dimension 1=@n13.Array Index
    ) -> (Output:"Struct Asset Copier Float Key Params Structure (by ref)")) -> (
    InterpolationMode_2_#43:ERichCurveInterpMode Enum
    TangentMode_18_#44:ERichCurveTangentMode Enum
    TangentWeightMode_14_#45:ERichCurveTangentWeightMode Enum
    ArriveTangent_7_#46:double=~
    ArriveTangentWeight_9_#47:double=~
    LeaveTangent_13_#48:double=~
    LeaveTangentWeight_17_#49:double=~
   ) {is=title K2Node_BreakStruct}
   > SetInterpolationMode!(execute, self:Movie Scene Scripting Float Key Object Reference=@n15, InNewValue:ERichCurveInterpMode Enum=@n16) -> (then)
   > SetTangentMode!(
    execute
    self:Movie Scene Scripting Float Key Object Reference=@n15
    InNewValue:ERichCurveTangentMode Enum=@n16.TangentMode_18_#44
   ) -> (then)
   > SetTangentWeightMode!(
    execute
    self:Movie Scene Scripting Float Key Object Reference=@n15
    InNewValue:ERichCurveTangentWeightMode Enum=@n16.TangentWeightMode_14_#45
   ) -> (then)
   > SetArriveTangent!(execute, self:Movie Scene Scripting Float Key Object Reference=@n15, InNewValue:float=@n16.ArriveTangent_7_#46) -> (then)
   > SetArriveTangentWeight!(
    execute
    self:Movie Scene Scripting Float Key Object Reference=@n15
    InNewValue:float=@n16.ArriveTangentWeight_9_#47
   ) -> (then)
   > SetLeaveTangent!(execute, self:Movie Scene Scripting Float Key Object Reference=@n15, InNewValue:float=@n16.LeaveTangent_13_#48) -> (then)
   > SetLeaveTangentWeight!(
    execute
    self:Movie Scene Scripting Float Key Object Reference=@n15
    InNewValue:float=@n16.LeaveTangentWeight_17_#49
   ) -> (then)
   "Set Default (Float)"!(
    execute=@n12.else
    self:Movie Scene Scripting Float Channel Object Reference=@n11
    InDefaultValue:float="Get (a ref)"(Array:"Array of Float (double-precision)s"=@n10.Array Element_FrameValue_33_#41, Dimension 1=0) -> (Output:"Float (double-precision) (by ref)")
   ) -> (then) {is=title K2Node_CallFunction}
  ]}
 SetTransformFloatKeys {nodes [
   n1 = SetTransformFloatKeys!() -> (then, Section:Movie Scene Section Object Reference, FloatData:Array of Struct Asset Copier Float Channel Structures) {is=title K2Node_FunctionEntry}
   > n2 = GetChannelsByType(Section=@n1, ChannelType=&MovieSceneScriptingDoubleChannel)
   > n3 = For Each Loop(Array:Array of Movie Scene Scripting Channel Object References=@n2) -> (Array Element:Movie Scene Scripting Channel Object Reference)
   > n4 = For Each Loop!(Exec, Array:Array of Struct Asset Copier Float Channel Structures=@n1.FloatData) -> (
    LoopBody
    Array Element_OwnerName_19_#39:String
    Array Element_FrameTime_27_#40:Array of Frame Time Structures
    Array Element_FrameValue_33_#41:"Array of Float (double-precision)s"
    Array Element_FloatKeyParams_32_#42:Array of Struct Asset Copier Float Key Params Structures
    Array Index
    Completed
   )
   > Branch(Condition=Contains(SearchIn=GetDisplayName(Object=@n3), Substring=@n4))
   > n5 = Cast To MovieSceneScriptingDoubleChannel(Object=@n3)
   > n6 = Sequence()
   n7 = Branch(execute=@n6.then_1, Condition="integer > integer"(A=Length(TargetArray=@n4.Array Element_FrameTime_27_#40)))
   > n8 = For Each Loop!(Exec, Array:Array of Frame Time Structures=@n4.Array Element_FrameTime_27_#40) -> (LoopBody, Array Element:Frame Time Structure, Array Index, Completed)
   n9 = Break Frame Time(FrameTime=@n8) -> (SubFrame=~)
   > n10 = "Add Key (Double)"!(
    execute
    self:Movie Scene Scripting Double Channel Object Reference=@n5
    InTime:"Frame Number Structure (by ref)"=FrameRateConversion(FrameNumber=@n9)
    NewValue:double="Get (a ref)"(Array=@n4.Array Element_FrameValue_33_#41, Dimension 1=@n8.Array Index)
    SubFrame:float=@n9.SubFrame
    TimeUnit:EMovieSceneTimeUnit Enum=DisplayRate
    InInterpolation:EMovieSceneKeyInterpolation Enum=Auto
   ) -> (then, ReturnValue:Movie Scene Scripting Double Key Object Reference)
   n11 = Break Struct Asset Copier Float Key Params(Struct_AssetCopier_FloatKeyParams="Get (a ref)"(
     Array:Array of Struct Asset Copier Float Key Params Structures=@n4.Array Element_FloatKeyParams_32_#42
     Dimension 1=@n8.Array Index
    ) -> (Output:"Struct Asset Copier Float Key Params Structure (by ref)")) -> (
    ArriveTangent_7_#46=~
    ArriveTangentWeight_9_#47=~
    LeaveTangent_13_#48=~
    LeaveTangentWeight_17_#49=~
   )
   > SetInterpolationMode(self:Movie Scene Scripting Double Key Object Reference=@n10, InNewValue=@n11)
   > SetTangentMode(self:Movie Scene Scripting Double Key Object Reference=@n10, InNewValue=@n11.TangentMode_18_#44)
   > SetTangentWeightMode(self:Movie Scene Scripting Double Key Object Reference=@n10, InNewValue=@n11.TangentWeightMode_14_#45)
   > SetArriveTangent(self:Movie Scene Scripting Double Key Object Reference=@n10, InNewValue=@n11.ArriveTangent_7_#46)
   > SetArriveTangentWeight(self:Movie Scene Scripting Double Key Object Reference=@n10, InNewValue=@n11.ArriveTangentWeight_9_#47)
   > SetLeaveTangent(self:Movie Scene Scripting Double Key Object Reference=@n10, InNewValue=@n11.LeaveTangent_13_#48)
   > SetLeaveTangentWeight(self:Movie Scene Scripting Double Key Object Reference=@n10, InNewValue=@n11.LeaveTangentWeight_17_#49)
   "Set Default (Double)"!(
    execute=@n7.else
    self:Movie Scene Scripting Double Channel Object Reference=@n5
    InDefaultValue:double="Get (a ref)"(Array:"Array of Float (double-precision)s"=@n4.Array Element_FrameValue_33_#41, Dimension 1=0) -> (Output:"Float (double-precision) (by ref)")
   ) -> (then) {is=title K2Node_CallFunction}
   n12 = "Get Keys (Double)"(self=@n5)
   n13 = For Each Loop(Exec=[@n14.then, @n6.then_0], Array:Array of Movie Scene Scripting Key Object References=@n12) -> (Array Element:Movie Scene Scripting Key Object Reference)
   n14 = Branch(execute=@n13.Completed, Condition="integer > integer"(A=Length(TargetArray:Array of Movie Scene Scripting Key Object References=@n12)))
   "Remove Key (Double)"!(execute=@n13.LoopBody, self:Movie Scene Scripting Double Channel Object Reference=@n5, Key:Movie Scene Scripting Key Object Reference=@n13) -> (then)
  ]}
 Sort Scroll Box By Checkbox {nodes [
   n1 = GetAllChildren(self=Get ScrollBox() -> (ScrollBox:Panel Widget Object Reference))
   n2 = For Each Loop(Exec=@n5.then, Array:Array of Widget Object References=@n1) -> (Array Element:Widget Object Reference)
   n3 = Cast To EUW_LevelSequenceAssetCopier_SourceRow(Object=@n2)
   n4 = Get CheckBox(self=@n3)
   > Branch(Condition="Not Equal (Enum)"!(
     A:ECheckBoxState Enum=Unchecked
     B:ECheckBoxState Enum=Get CheckedState!(self:Check Box Object Reference=@n4) -> (CheckedState:ECheckBoxState Enum=Unchecked)
    ) -> (ReturnValue:bool))
   > Add(
    TargetArray:Array of EUW Level Sequence Asset Copier Source Row Object References=Get Items To Sort() -> (Items To Sort:Array of EUW Level Sequence Asset Copier Source Row Object References)
    NewItem:"EUW Level Sequence Asset Copier Source Row Object Reference (by ref)"=@n3
   ) {is=title K2Node_CallArrayFunction}
   Sort Scroll Box By Checkbox!() -> (then, ScrollBox:Panel Widget Object Reference) {is=title K2Node_FunctionEntry}
   > n5 = Clear(TargetArray:Array of EUW Level Sequence Asset Copier Source Row Object References=Get Items To Sort()) {is=title K2Node_CallArrayFunction}
   n6 = For Each Loop(Exec=@n2.Completed, Array=@n1)
   > n7 = Cast To EUW_LevelSequenceAssetCopier_SourceRow(Object=@n6)
   Add(execute=@n11.then, TargetArray=Get Items To Sort(), NewItem=@n7)
   n8 = For Each Loop(Exec=@n9.then, Array:Array of EUW Level Sequence Asset Copier Source Row Object References=Get Items To Sort()) -> (Array Element:EUW Level Sequence Asset Copier Source Row Object Reference)
   > AddChild(self=Get ScrollBox(), Content=@n8)
   n9 = ClearChildren(execute=@n6.Completed, self=Get ScrollBox())
   n10 = "Not Equal (Enum)"(A=Get CheckedState(self=Get CheckBox(self=@n7)) -> (CheckedState=Unchecked), B=Checked)
   n11 = Branch(execute=@n7.then, Condition=@n10)
  ]}
 UpVersion {nodes [
   n1 = UpVersion!() -> (then, InName:String, NameList:Array of Strings) {is=title K2Node_FunctionEntry}
   > n2 = Set CheckName(CheckName:String=@n1) -> (Output_Get:String)
   n3 = Get CheckName() -> (CheckName:String)
   n4 = Branch(
    execute=[@n2.then, @n8.then]
    Condition="integer > integer"(A=Find Item(TargetArray:Array of Strings=@n1.NameList, ItemToFind:"String (by ref)"=@n3))
   )
   > Increment Int(Value=Get VersionInt() -> (VersionInt:int=0))
   n5 = "To String (Integer)"(InInt=Get VersionInt() -> (VersionInt=0))
   n6 = Get InName() -> (InName:String)
   n7 = Get CheckName()
   > n8 = Set CheckName(CheckName=Replace(SourceString=@n7, From=@n7, To=Append!(A=@n6, B=_, C:String=@n5) -> (ReturnValue)))
   n9 = Branch(execute=@n4.else, Condition="Equal (Integer)"(A=Get VersionInt() -> (VersionInt=0)))
   > n10 = Set NewName(NewName:String=Get InName()) -> (Output_Get:String)
   Return Node!(execute=[@n10.then, @n11.then], OutName:String=Get NewName() -> (NewName:String))
   n11 = Set NewName(execute=@n9.else, NewName=Get CheckName())
  ]}
 addSpawnedSection {nodes [
   n1 = addSpawnedSection!() -> (then, Track:Movie Scene Track Object Reference, BoolChannels:Array of Struct Asset Copier Bool Channel Structures) {is=title K2Node_FunctionEntry}
   > n2 = AddSection(Track=@n1)
   > n3 = Cast To MovieSceneBoolSection(Object=@n2)
   > SetStartFrameBounded(Section=@n3, bIsBounded=false)
   > SetEndFrameBounded(Section=@n3, bIsBounded=false)
   > SetBoolKeys(Section=@n3, BoolChannels=Get BoolChannels()) {is=title K2Node_CallFunction}
  ]}
]
variables [
 BeenOpened {type="array<LevelSequence>", default []}
 CopiedBinding {type=string, default=""}
 Current Binding {type=MovieSceneBindingProxy, default {bindingId=#56, sequence=null}}
 Current Project {type=BP_AnimationBrowser_Project_C, default=null}
 CurrentProjects {type="array<BP_AnimationBrowser_Project_C>", default []}
 EUW Rows {type="array<EUW_LevelSequenceAssetCopier_SourceRow_C>", default []}
 Out Data {
  type=Struct_AssetCopier_BindingData
  default {
   bindingOwnerName=""
   objectTemplate=null
   overwrite=false
   parentAssetName=""
   animTracks []
   boolTracks []
   childBindings []
   controlRigBinding {controlRig=null, track=null, proxy {bindingId=#56, sequence=null}}
   controlRigTracks []
   floatTracks []
   spawnTracks []
   transformTrack {propertyName=None, propertyPath="", transformSections []}
  }
 }
 ProjectName {type=string, default=""}
 ProjectNames {type="array<string>", default []}
 ShotNumber {type=string, default=""}
 Source Level Sequence {type=LevelSequence, default=null}
 SourceBindingData {
  type=Struct_AssetCopier_BindingData
  default {
   bindingOwnerName=""
   objectTemplate=null
   overwrite=false
   parentAssetName=""
   animTracks []
   boolTracks []
   childBindings []
   controlRigBinding {controlRig=null, track=null, proxy {bindingId=#56, sequence=null}}
   controlRigTracks []
   floatTracks []
   spawnTracks []
   transformTrack {propertyName=None, propertyPath="", transformSections []}
  }
 }
 SubFolderName {type=string, default=""}
 SubShot {type=string, default=""}
 Target Level Sequence {type=LevelSequence, default=null}
 Timer {type=TimerHandle}
]
widgets {
 info {rootWidgetClass=&CanvasPanel, widgetCount=17}
 widgets [
  CanvasPanel_0 {widgetClassPath=&CanvasPanel, widget=CanvasPanel_0}
  VerticalBox_0 {
   widgetClassPath=&VerticalBox
   parent=CanvasPanel_0
   widget=VerticalBox_0
   slot {class=CanvasPanelSlot, properties {layoutData {alignment=<0 0>, anchors {maximum=<1 1>, minimum=<0 0>}, offsets=<ltrb 0 0 0 0>}}}
  }
  HorizontalBox_0 {widgetClassPath=&HorizontalBox, parent=VerticalBox_0, widget=HorizontalBox_0, slot {class=VerticalBoxSlot}}
  CurrentLevelSequenceButton {
   widgetClassPath=&EditorUtilityButton
   parent=HorizontalBox_0
   widget=CurrentLevelSequenceButton
   properties {visibility=Collapsed}
   slot {class=HorizontalBoxSlot, properties {horizontalAlignment=HAlign_Center}}
  }
  TextBlock_106 {
   widgetClassPath=&TextBlock
   parent=CurrentLevelSequenceButton
   widget=TextBlock_106
   properties {
    text=Current LS
    font {
     bForceMonospaced=false
     bMaterialIsStencil=false
     fontMaterial=null
     fontObject=&Roboto
     letterSpacing=0
     monospacedWidth=1
     size=9
     skewAmount=0
     typefaceFontName=Bold
     outlineSettings {
      bApplyOutlineToDropShadows=false
      bMiteredCorners=false
      bSeparateFillAlpha=false
      outlineMaterial=null
      outlineSize=0
      outlineColor=<rgba 0 0 0 1>
     }
    }
   }
   slot {class=ButtonSlot}
  }
  VerticalBox_80 {
   widgetClassPath=&VerticalBox
   parent=HorizontalBox_0
   widget=VerticalBox_80
   slot {class=HorizontalBoxSlot, properties {size {sizeRule=Fill, value=1}}}
  }
  TextBlock_65 {
   widgetClassPath=&TextBlock
   parent=VerticalBox_80
   widget=TextBlock_65
   properties {text=Source Sequence}
   slot {class=VerticalBoxSlot, properties {horizontalAlignment=HAlign_Center, verticalAlignment=VAlign_Center, size {sizeRule=Fill, value=1}}}
  }
  SourceLevelSequencePropertyView {
   widgetClassPath=&SinglePropertyView
   parent=VerticalBox_80
   widget=SourceLevelSequencePropertyView
   properties {propertyName=Source Level Sequence}
   slot {class=VerticalBoxSlot, properties {size {sizeRule=Fill, value=2.5}}}
  }
  Image_57 {
   widgetClassPath=&Image
   bIsVariable=true
   parent=HorizontalBox_0
   widget=Image_57
   properties {brush {
     bIsDynamicallyLoaded=false
     drawAs=Image
     imageType=FullColor
     mirroring=NoMirror
     resourceName=None
     resourceObject=/Game/NeverCook/EditorUtilities/DebugTools/AssetChangeFinder/Art/ArrowRight
     tiling=NoTile
     imageSize=<64 64>
     margin=<ltrb 0 0 0 0>
     outlineSettings {
      bUseBrushTransparency=false
      roundingType=HalfHeightRadius
      width=0
      color {colorUseRule=UseColor_Specified, specifiedColor=<rgba 0 0 0 0>}
      cornerRadii=<0 0 0 0>
     }
     tintColor {colorUseRule=UseColor_Specified, specifiedColor=<rgba 1 1 1 1>}
     uVRegion {bIsValid=false, max=<0 0>, min=<0 0>}
    }}
   slot {class=HorizontalBoxSlot, properties {verticalAlignment=VAlign_Center}}
  }
  VerticalBox_191 {
   widgetClassPath=&VerticalBox
   parent=HorizontalBox_0
   widget=VerticalBox_191
   slot {class=HorizontalBoxSlot, properties {size {sizeRule=Fill, value=1}}}
  }
  TextBlock {
   widgetClassPath=&TextBlock
   parent=VerticalBox_191
   widget=TextBlock
   properties {text=Target Sequence}
   slot {class=VerticalBoxSlot, properties {horizontalAlignment=HAlign_Center, verticalAlignment=VAlign_Center, size {sizeRule=Fill, value=1}}}
  }
  TargetLevelSequencePropertyView {
   widgetClassPath=&SinglePropertyView
   parent=VerticalBox_191
   widget=TargetLevelSequencePropertyView
   properties {propertyName=Target Level Sequence}
   slot {class=VerticalBoxSlot, properties {padding=<ltrb 20 0 0 0>, size {sizeRule=Fill, value=2.5}}}
  }
  HorizontalBox_183 {
   widgetClassPath=&HorizontalBox
   parent=CanvasPanel_0
   widget=HorizontalBox_183
   slot {
    class=CanvasPanelSlot
    properties {bAutoSize=true, layoutData {alignment=<0 0>, anchors {maximum=<1 1>, minimum=<0 0>}, offsets=<ltrb 0 92.1283 0 52>}}
   }
  }
  SourceScrollBox {
   widgetClassPath=&ScrollBox
   bIsVariable=true
   parent=HorizontalBox_183
   widget=SourceScrollBox
   properties {
    allowOverscroll=false
    alwaysShowScrollbar=true
    consumeMouseWheel=Always
    scrollWhenFocusChanges=AnimatedScroll
    widgetBarStyle {
     thickness=8
     draggedThumbImage {
      bIsDynamicallyLoaded=false
      drawAs=RoundedBox
      imageType=NoImage
      mirroring=NoMirror
      resourceName=None
      resourceObject=null
      tiling=NoTile
      imageSize=<0 0>
      margin=<ltrb 0 0 0 0>
      outlineSettings {
       bUseBrushTransparency=false
       roundingType=FixedRadius
       width=0
       color {colorUseRule=UseColor_Specified, specifiedColor=<rgba 0 0 0 0>}
       cornerRadii=<4 4 4 4>
      }
      tintColor {colorUseRule=UseColor_Specified, specifiedColor=<rgba 0.2159 0.2159 0.2159 1>}
      uVRegion {bIsValid=false, max=<0 0>, min=<0 0>}
     }
     horizontalBackgroundImage {
      bIsDynamicallyLoaded=false
      drawAs=NoDrawType
      imageType=NoImage
      mirroring=NoMirror
      resourceName=None
      resourceObject=null
      tiling=NoTile
      imageSize=<0 0>
      margin=<ltrb 0 0 0 0>
      outlineSettings {
       bUseBrushTransparency=false
       roundingType=HalfHeightRadius
       width=0
       color {colorUseRule=UseColor_Specified, specifiedColor=<rgba 0 0 0 0>}
       cornerRadii=<0 0 0 0>
      }
      tintColor {colorUseRule=UseColor_Specified, specifiedColor=<rgba 1 1 1 1>}
      uVRegion {bIsValid=false, max=<0 0>, min=<0 0>}
     }
     horizontalBottomSlotImage {
      bIsDynamicallyLoaded=false
      drawAs=NoDrawType
      imageType=NoImage
      mirroring=NoMirror
      resourceName=None
      resourceObject=null
      tiling=NoTile
      imageSize=<0 0>
      margin=<ltrb 0 0 0 0>
      outlineSettings {
       bUseBrushTransparency=false
       roundingType=HalfHeightRadius
       width=0
       color {colorUseRule=UseColor_Specified, specifiedColor=<rgba 0 0 0 0>}
       cornerRadii=<0 0 0 0>
      }
      tintColor {colorUseRule=UseColor_Specified, specifiedColor=<rgba 1 1 1 1>}
      uVRegion {bIsValid=false, max=<0 0>, min=<0 0>}
     }
     horizontalTopSlotImage {
      bIsDynamicallyLoaded=false
      drawAs=NoDrawType
      imageType=NoImage
      mirroring=NoMirror
      resourceName=None
      resourceObject=null
      tiling=NoTile
      imageSize=<0 0>
      margin=<ltrb 0 0 0 0>
      outlineSettings {
       bUseBrushTransparency=false
       roundingType=HalfHeightRadius
       width=0
       color {colorUseRule=UseColor_Specified, specifiedColor=<rgba 0 0 0 0>}
       cornerRadii=<0 0 0 0>
      }
      tintColor {colorUseRule=UseColor_Specified, specifiedColor=<rgba 1 1 1 1>}
      uVRegion {bIsValid=false, max=<0 0>, min=<0 0>}
     }
     hoveredThumbImage {
      bIsDynamicallyLoaded=false
      drawAs=RoundedBox
      imageType=NoImage
      mirroring=NoMirror
      resourceName=None
      resourceObject=null
      tiling=NoTile
      imageSize=<0 0>
      margin=<ltrb 0 0 0 0>
      outlineSettings {
       bUseBrushTransparency=false
       roundingType=FixedRadius
       width=0
       color {colorUseRule=UseColor_Specified, specifiedColor=<rgba 0 0 0 0>}
       cornerRadii=<4 4 4 4>
      }
      tintColor {colorUseRule=UseColor_Specified, specifiedColor=<rgba 0.2159 0.2159 0.2159 1>}
      uVRegion {bIsValid=false, max=<0 0>, min=<0 0>}
     }
     normalThumbImage {
      bIsDynamicallyLoaded=false
      drawAs=RoundedBox
      imageType=NoImage
      mirroring=NoMirror
      resourceName=None
      resourceObject=null
      tiling=NoTile
      imageSize=<0 0>
      margin=<ltrb 0 0 0 0>
      outlineSettings {
       bUseBrushTransparency=false
       roundingType=FixedRadius
       width=0
       color {colorUseRule=UseColor_Specified, specifiedColor=<rgba 0 0 0 0>}
       cornerRadii=<4 4 4 4>
      }
      tintColor {colorUseRule=UseColor_Specified, specifiedColor=<rgba 0.0953 0.0953 0.0953 1>}
      uVRegion {bIsValid=false, max=<0 0>, min=<0 0>}
     }
     verticalBackgroundImage {
      bIsDynamicallyLoaded=false
      drawAs=NoDrawType
      imageType=NoImage
      mirroring=NoMirror
      resourceName=None
      resourceObject=null
      tiling=NoTile
      imageSize=<0 0>
      margin=<ltrb 0 0 0 0>
      outlineSettings {
       bUseBrushTransparency=false
       roundingType=HalfHeightRadius
       width=0
       color {colorUseRule=UseColor_Specified, specifiedColor=<rgba 0 0 0 0>}
       cornerRadii=<0 0 0 0>
      }
      tintColor {colorUseRule=UseColor_Specified, specifiedColor=<rgba 1 1 1 1>}
      uVRegion {bIsValid=false, max=<0 0>, min=<0 0>}
     }
     verticalBottomSlotImage {
      bIsDynamicallyLoaded=false
      drawAs=NoDrawType
      imageType=NoImage
      mirroring=NoMirror
      resourceName=None
      resourceObject=null
      tiling=NoTile
      imageSize=<0 0>
      margin=<ltrb 0 0 0 0>
      outlineSettings {
       bUseBrushTransparency=false
       roundingType=HalfHeightRadius
       width=0
       color {colorUseRule=UseColor_Specified, specifiedColor=<rgba 0 0 0 0>}
       cornerRadii=<0 0 0 0>
      }
      tintColor {colorUseRule=UseColor_Specified, specifiedColor=<rgba 1 1 1 1>}
      uVRegion {bIsValid=false, max=<0 0>, min=<0 0>}
     }
     verticalTopSlotImage {
      bIsDynamicallyLoaded=false
      drawAs=NoDrawType
      imageType=NoImage
      mirroring=NoMirror
      resourceName=None
      resourceObject=null
      tiling=NoTile
      imageSize=<0 0>
      margin=<ltrb 0 0 0 0>
      outlineSettings {
       bUseBrushTransparency=false
       roundingType=HalfHeightRadius
       width=0
       color {colorUseRule=UseColor_Specified, specifiedColor=<rgba 0 0 0 0>}
       cornerRadii=<0 0 0 0>
      }
      tintColor {colorUseRule=UseColor_Specified, specifiedColor=<rgba 1 1 1 1>}
      uVRegion {bIsValid=false, max=<0 0>, min=<0 0>}
     }
    }
    widgetStyle {
     barThickness=8
     bottomShadowBrush {
      bIsDynamicallyLoaded=false
      drawAs=Box
      imageType=FullColor
      mirroring=NoMirror
      resourceName=../../../Engine/Content/Slate/Common/ScrollBoxShadowBottom.png
      resourceObject=null
      tiling=NoTile
      imageSize=<16 8>
      margin=<ltrb 0.5 0 0.5 1>
      outlineSettings {
       bUseBrushTransparency=false
       roundingType=HalfHeightRadius
       width=0
       color {colorUseRule=UseColor_Specified, specifiedColor=<rgba 0 0 0 0>}
       cornerRadii=<0 0 0 0>
      }
      tintColor {colorUseRule=UseColor_Specified, specifiedColor=<rgba 1 1 1 1>}
      uVRegion {bIsValid=false, max=<0 0>, min=<0 0>}
     }
     horizontalScrolledContentPadding=<ltrb 0 0 1 0>
     leftShadowBrush {
      bIsDynamicallyLoaded=false
      drawAs=Box
      imageType=FullColor
      mirroring=NoMirror
      resourceName=../../../Engine/Content/Slate/Common/ScrollBoxShadowLeft.png
      resourceObject=null
      tiling=NoTile
      imageSize=<8 16>
      margin=<ltrb 1 0.5 0 0.5>
      outlineSettings {
       bUseBrushTransparency=false
       roundingType=HalfHeightRadius
       width=0
       color {colorUseRule=UseColor_Specified, specifiedColor=<rgba 0 0 0 0>}
       cornerRadii=<0 0 0 0>
      }
      tintColor {colorUseRule=UseColor_Specified, specifiedColor=<rgba 1 1 1 1>}
      uVRegion {bIsValid=false, max=<0 0>, min=<0 0>}
     }
     rightShadowBrush {
      bIsDynamicallyLoaded=false
      drawAs=Box
      imageType=FullColor
      mirroring=NoMirror
      resourceName=../../../Engine/Content/Slate/Common/ScrollBoxShadowRight.png
      resourceObject=null
      tiling=NoTile
      imageSize=<8 16>
      margin=<ltrb 0 0.5 1 0.5>
      outlineSettings {
       bUseBrushTransparency=false
       roundingType=HalfHeightRadius
       width=0
       color {colorUseRule=UseColor_Specified, specifiedColor=<rgba 0 0 0 0>}
       cornerRadii=<0 0 0 0>
      }
      tintColor {colorUseRule=UseColor_Specified, specifiedColor=<rgba 1 1 1 1>}
      uVRegion {bIsValid=false, max=<0 0>, min=<0 0>}
     }
     topShadowBrush {
      bIsDynamicallyLoaded=false
      drawAs=Box
      imageType=FullColor
      mirroring=NoMirror
      resourceName=../../../Engine/Content/Slate/Common/ScrollBoxShadowTop.png
      resourceObject=null
      tiling=NoTile
      imageSize=<16 8>
      margin=<ltrb 0.5 1 0.5 0>
      outlineSettings {
       bUseBrushTransparency=false
       roundingType=HalfHeightRadius
       width=0
       color {colorUseRule=UseColor_Specified, specifiedColor=<rgba 0 0 0 0>}
       cornerRadii=<0 0 0 0>
      }
      tintColor {colorUseRule=UseColor_Specified, specifiedColor=<rgba 1 1 1 1>}
      uVRegion {bIsValid=false, max=<0 0>, min=<0 0>}
     }
     verticalScrolledContentPadding=<ltrb 0 0 0 1>
    }
   }
   slot {class=HorizontalBoxSlot, properties {size {sizeRule=Fill, value=1}}}
  }
  TargetScrollBox {
   widgetClassPath=&ScrollBox
   bIsVariable=true
   parent=HorizontalBox_183
   widget=TargetScrollBox
   properties {
    consumeMouseWheel=Always
    widgetBarStyle {
     thickness=8
     draggedThumbImage {
      bIsDynamicallyLoaded=false
      drawAs=RoundedBox
      imageType=NoImage
      mirroring=NoMirror
      resourceName=None
      resourceObject=null
      tiling=NoTile
      imageSize=<0 0>
      margin=<ltrb 0 0 0 0>
      outlineSettings {
       bUseBrushTransparency=false
       roundingType=FixedRadius
       width=0
       color {colorUseRule=UseColor_Specified, specifiedColor=<rgba 0 0 0 0>}
       cornerRadii=<4 4 4 4>
      }
      tintColor {colorUseRule=UseColor_Specified, specifiedColor=<rgba 0.2159 0.2159 0.2159 1>}
      uVRegion {bIsValid=false, max=<0 0>, min=<0 0>}
     }
     horizontalBackgroundImage {
      bIsDynamicallyLoaded=false
      drawAs=NoDrawType
      imageType=NoImage
      mirroring=NoMirror
      resourceName=None
      resourceObject=null
      tiling=NoTile
      imageSize=<0 0>
      margin=<ltrb 0 0 0 0>
      outlineSettings {
       bUseBrushTransparency=false
       roundingType=HalfHeightRadius
       width=0
       color {colorUseRule=UseColor_Specified, specifiedColor=<rgba 0 0 0 0>}
       cornerRadii=<0 0 0 0>
      }
      tintColor {colorUseRule=UseColor_Specified, specifiedColor=<rgba 1 1 1 1>}
      uVRegion {bIsValid=false, max=<0 0>, min=<0 0>}
     }
     horizontalBottomSlotImage {
      bIsDynamicallyLoaded=false
      drawAs=NoDrawType
      imageType=NoImage
      mirroring=NoMirror
      resourceName=None
      resourceObject=null
      tiling=NoTile
      imageSize=<0 0>
      margin=<ltrb 0 0 0 0>
      outlineSettings {
       bUseBrushTransparency=false
       roundingType=HalfHeightRadius
       width=0
       color {colorUseRule=UseColor_Specified, specifiedColor=<rgba 0 0 0 0>}
       cornerRadii=<0 0 0 0>
      }
      tintColor {colorUseRule=UseColor_Specified, specifiedColor=<rgba 1 1 1 1>}
      uVRegion {bIsValid=false, max=<0 0>, min=<0 0>}
     }
     horizontalTopSlotImage {
      bIsDynamicallyLoaded=false
      drawAs=NoDrawType
      imageType=NoImage
      mirroring=NoMirror
      resourceName=None
      resourceObject=null
      tiling=NoTile
      imageSize=<0 0>
      margin=<ltrb 0 0 0 0>
      outlineSettings {
       bUseBrushTransparency=false
       roundingType=HalfHeightRadius
       width=0
       color {colorUseRule=UseColor_Specified, specifiedColor=<rgba 0 0 0 0>}
       cornerRadii=<0 0 0 0>
      }
      tintColor {colorUseRule=UseColor_Specified, specifiedColor=<rgba 1 1 1 1>}
      uVRegion {bIsValid=false, max=<0 0>, min=<0 0>}
     }
     hoveredThumbImage {
      bIsDynamicallyLoaded=false
      drawAs=RoundedBox
      imageType=NoImage
      mirroring=NoMirror
      resourceName=None
      resourceObject=null
      tiling=NoTile
      imageSize=<0 0>
      margin=<ltrb 0 0 0 0>
      outlineSettings {
       bUseBrushTransparency=false
       roundingType=FixedRadius
       width=0
       color {colorUseRule=UseColor_Specified, specifiedColor=<rgba 0 0 0 0>}
       cornerRadii=<4 4 4 4>
      }
      tintColor {colorUseRule=UseColor_Specified, specifiedColor=<rgba 0.2159 0.2159 0.2159 1>}
      uVRegion {bIsValid=false, max=<0 0>, min=<0 0>}
     }
     normalThumbImage {
      bIsDynamicallyLoaded=false
      drawAs=RoundedBox
      imageType=NoImage
      mirroring=NoMirror
      resourceName=None
      resourceObject=null
      tiling=NoTile
      imageSize=<0 0>
      margin=<ltrb 0 0 0 0>
      outlineSettings {
       bUseBrushTransparency=false
       roundingType=FixedRadius
       width=0
       color {colorUseRule=UseColor_Specified, specifiedColor=<rgba 0 0 0 0>}
       cornerRadii=<4 4 4 4>
      }
      tintColor {colorUseRule=UseColor_Specified, specifiedColor=<rgba 0.0953 0.0953 0.0953 1>}
      uVRegion {bIsValid=false, max=<0 0>, min=<0 0>}
     }
     verticalBackgroundImage {
      bIsDynamicallyLoaded=false
      drawAs=NoDrawType
      imageType=NoImage
      mirroring=NoMirror
      resourceName=None
      resourceObject=null
      tiling=NoTile
      imageSize=<0 0>
      margin=<ltrb 0 0 0 0>
      outlineSettings {
       bUseBrushTransparency=false
       roundingType=HalfHeightRadius
       width=0
       color {colorUseRule=UseColor_Specified, specifiedColor=<rgba 0 0 0 0>}
       cornerRadii=<0 0 0 0>
      }
      tintColor {colorUseRule=UseColor_Specified, specifiedColor=<rgba 1 1 1 1>}
      uVRegion {bIsValid=false, max=<0 0>, min=<0 0>}
     }
     verticalBottomSlotImage {
      bIsDynamicallyLoaded=false
      drawAs=NoDrawType
      imageType=NoImage
      mirroring=NoMirror
      resourceName=None
      resourceObject=null
      tiling=NoTile
      imageSize=<0 0>
      margin=<ltrb 0 0 0 0>
      outlineSettings {
       bUseBrushTransparency=false
       roundingType=HalfHeightRadius
       width=0
       color {colorUseRule=UseColor_Specified, specifiedColor=<rgba 0 0 0 0>}
       cornerRadii=<0 0 0 0>
      }
      tintColor {colorUseRule=UseColor_Specified, specifiedColor=<rgba 1 1 1 1>}
      uVRegion {bIsValid=false, max=<0 0>, min=<0 0>}
     }
     verticalTopSlotImage {
      bIsDynamicallyLoaded=false
      drawAs=NoDrawType
      imageType=NoImage
      mirroring=NoMirror
      resourceName=None
      resourceObject=null
      tiling=NoTile
      imageSize=<0 0>
      margin=<ltrb 0 0 0 0>
      outlineSettings {
       bUseBrushTransparency=false
       roundingType=HalfHeightRadius
       width=0
       color {colorUseRule=UseColor_Specified, specifiedColor=<rgba 0 0 0 0>}
       cornerRadii=<0 0 0 0>
      }
      tintColor {colorUseRule=UseColor_Specified, specifiedColor=<rgba 1 1 1 1>}
      uVRegion {bIsValid=false, max=<0 0>, min=<0 0>}
     }
    }
    widgetStyle {
     barThickness=8
     bottomShadowBrush {
      bIsDynamicallyLoaded=false
      drawAs=Box
      imageType=FullColor
      mirroring=NoMirror
      resourceName=../../../Engine/Content/Slate/Common/ScrollBoxShadowBottom.png
      resourceObject=null
      tiling=NoTile
      imageSize=<16 8>
      margin=<ltrb 0.5 0 0.5 1>
      outlineSettings {
       bUseBrushTransparency=false
       roundingType=HalfHeightRadius
       width=0
       color {colorUseRule=UseColor_Specified, specifiedColor=<rgba 0 0 0 0>}
       cornerRadii=<0 0 0 0>
      }
      tintColor {colorUseRule=UseColor_Specified, specifiedColor=<rgba 1 1 1 1>}
      uVRegion {bIsValid=false, max=<0 0>, min=<0 0>}
     }
     horizontalScrolledContentPadding=<ltrb 0 0 1 0>
     leftShadowBrush {
      bIsDynamicallyLoaded=false
      drawAs=Box
      imageType=FullColor
      mirroring=NoMirror
      resourceName=../../../Engine/Content/Slate/Common/ScrollBoxShadowLeft.png
      resourceObject=null
      tiling=NoTile
      imageSize=<8 16>
      margin=<ltrb 1 0.5 0 0.5>
      outlineSettings {
       bUseBrushTransparency=false
       roundingType=HalfHeightRadius
       width=0
       color {colorUseRule=UseColor_Specified, specifiedColor=<rgba 0 0 0 0>}
       cornerRadii=<0 0 0 0>
      }
      tintColor {colorUseRule=UseColor_Specified, specifiedColor=<rgba 1 1 1 1>}
      uVRegion {bIsValid=false, max=<0 0>, min=<0 0>}
     }
     rightShadowBrush {
      bIsDynamicallyLoaded=false
      drawAs=Box
      imageType=FullColor
      mirroring=NoMirror
      resourceName=../../../Engine/Content/Slate/Common/ScrollBoxShadowRight.png
      resourceObject=null
      tiling=NoTile
      imageSize=<8 16>
      margin=<ltrb 0 0.5 1 0.5>
      outlineSettings {
       bUseBrushTransparency=false
       roundingType=HalfHeightRadius
       width=0
       color {colorUseRule=UseColor_Specified, specifiedColor=<rgba 0 0 0 0>}
       cornerRadii=<0 0 0 0>
      }
      tintColor {colorUseRule=UseColor_Specified, specifiedColor=<rgba 1 1 1 1>}
      uVRegion {bIsValid=false, max=<0 0>, min=<0 0>}
     }
     topShadowBrush {
      bIsDynamicallyLoaded=false
      drawAs=Box
      imageType=FullColor
      mirroring=NoMirror
      resourceName=../../../Engine/Content/Slate/Common/ScrollBoxShadowTop.png
      resourceObject=null
      tiling=NoTile
      imageSize=<16 8>
      margin=<ltrb 0.5 1 0.5 0>
      outlineSettings {
       bUseBrushTransparency=false
       roundingType=HalfHeightRadius
       width=0
       color {colorUseRule=UseColor_Specified, specifiedColor=<rgba 0 0 0 0>}
       cornerRadii=<0 0 0 0>
      }
      tintColor {colorUseRule=UseColor_Specified, specifiedColor=<rgba 1 1 1 1>}
      uVRegion {bIsValid=false, max=<0 0>, min=<0 0>}
     }
     verticalScrolledContentPadding=<ltrb 0 0 0 1>
    }
   }
   slot {class=HorizontalBoxSlot, properties {size {sizeRule=Fill, value=1}}}
  }
  CopyButton {
   widgetClassPath=&Button
   parent=CanvasPanel_0
   widget=CopyButton
   properties {widgetStyle {
     clickedSlateSound {resourceObject=null}
     disabled {
      bIsDynamicallyLoaded=false
      drawAs=RoundedBox
      imageType=NoImage
      mirroring=NoMirror
      resourceName=None
      resourceObject=null
      tiling=NoTile
      imageSize=<0 0>
      margin=<ltrb 0 0 0 0>
      outlineSettings {
       bUseBrushTransparency=false
       roundingType=FixedRadius
       width=1
       color {colorUseRule=UseColor_ColorTable, specifiedColor=<rgba 0 0 0 0>}
       cornerRadii=<4 4 4 4>
      }
      tintColor {colorUseRule=UseColor_Specified, specifiedColor=<rgba 0.0395 0.0395 0.0395 1>}
      uVRegion {bIsValid=false, max=<0 0>, min=<0 0>}
     }
     disabledForeground {colorUseRule=UseColor_ColorTable, specifiedColor=<rgba 0 0 0 0>}
     hovered {
      bIsDynamicallyLoaded=false
      drawAs=RoundedBox
      imageType=NoImage
      mirroring=NoMirror
      resourceName=None
      resourceObject=null
      tiling=NoTile
      imageSize=<32 32>
      margin=<ltrb 0 0 0 0>
      outlineSettings {
       bUseBrushTransparency=false
       roundingType=FixedRadius
       width=1
       color {colorUseRule=UseColor_ColorTable, specifiedColor=<rgba 0 0 0 0>}
       cornerRadii=<4 4 4 4>
      }
      tintColor {colorUseRule=UseColor_Specified, specifiedColor=<rgba 0.0953 0.0953 0.0953 1>}
      uVRegion {bIsValid=false, max=<0 0>, min=<0 0>}
     }
     hoveredForeground {colorUseRule=UseColor_ColorTable, specifiedColor=<rgba 0 0 0 0>}
     hoveredSlateSound {resourceObject=null}
     normal {
      bIsDynamicallyLoaded=false
      drawAs=RoundedBox
      imageType=NoImage
      mirroring=NoMirror
      resourceName=None
      resourceObject=null
      tiling=NoTile
      imageSize=<32 32>
      margin=<ltrb 0 0 0 0>
      outlineSettings {
       bUseBrushTransparency=false
       roundingType=FixedRadius
       width=1
       color {colorUseRule=UseColor_ColorTable, specifiedColor=<rgba 0 0 0 0>}
       cornerRadii=<4 4 4 4>
      }
      tintColor {colorUseRule=UseColor_Specified, specifiedColor=<rgba 0.0395 0.0395 0.0395 1>}
      uVRegion {bIsValid=false, max=<0 0>, min=<0 0>}
     }
     normalForeground {colorUseRule=UseColor_ColorTable, specifiedColor=<rgba 0 0 0 0>}
     normalPadding=<ltrb 12 1.5 12 1.5>
     pressed {
      bIsDynamicallyLoaded=false
      drawAs=RoundedBox
      imageType=NoImage
      mirroring=NoMirror
      resourceName=None
      resourceObject=null
      tiling=NoTile
      imageSize=<32 32>
      margin=<ltrb 0 0 0 0>
      outlineSettings {
       bUseBrushTransparency=false
       roundingType=FixedRadius
       width=1
       color {colorUseRule=UseColor_ColorTable, specifiedColor=<rgba 0 0 0 0>}
       cornerRadii=<4 4 4 4>
      }
      tintColor {colorUseRule=UseColor_Specified, specifiedColor=<rgba 0.0284 0.0284 0.0284 1>}
      uVRegion {bIsValid=false, max=<0 0>, min=<0 0>}
     }
     pressedForeground {colorUseRule=UseColor_ColorTable, specifiedColor=<rgba 0 0 0 0>}
     pressedPadding=<ltrb 12 2.5 12 0.5>
     pressedSlateSound {resourceObject=null}
    }}
   slot {
    class=CanvasPanelSlot
    properties {bAutoSize=true, zOrder=1, layoutData {alignment=<0 0>, anchors {maximum=<1 1>, minimum=<0 1>}, offsets=<ltrb 0 -45 0 30>}}
   }
  }
  TextBlock_4 {widgetClassPath=&TextBlock, parent=CopyButton, widget=TextBlock_4, properties {text=Copy}, slot {class=ButtonSlot}}
 ]
}
